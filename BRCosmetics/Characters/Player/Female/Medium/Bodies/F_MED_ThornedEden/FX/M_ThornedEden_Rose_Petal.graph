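# Masked, two-sided subsurface material for Niagara sprite and mesh particles, drawing rose petals for the Thorned Eden character's FX.
# A custom HLSL polar curve and sphere masks carve the petal shape procedurally on bump-offset, warped UVs.
# Color parameters blend with those masks and a swirl texture, varied by particle random or a dynamic parameter.
# Particle color tints the result, with sphere-normal texture and scaled subsurface terms adding volume.
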Material /BRCosmetics/Characters/Player/Female/Medium/Bodies/F_MED_ThornedEden/FX/M_ThornedEden_Rose_Petal
graph {
 properties {
  bUsedWithNiagaraMeshParticles=true
  bUsedWithNiagaraSprites=true
  blendMode=BLEND_Masked
  opacityMaskClipValue=0.1
  pixelDepthOffsetMode=PDOM_Legacy
  shadingModel=MSM_Subsurface
  twoSided=true
  usedShadingModels=Subsurface
 }
 nodes [
  n1 = TextureCoordinate()
  n2 = Constant() {r=0.75}
  n3 = Constant() {r=0.5}
  n4 = SphereMask(A=TextureCoordinate(), B=@n3, Radius=@n3, Hardness=Constant())
  n5 = MaterialFunctionCall(Float2=ShadingPathSwitch(Default=BumpOffset(Coordinate=@n1, Height=@n2, HeightRatioInput=@n4), Mobile=@n1) {desc=Migrated from FeatureLevelSwitch}) {materialFunction=&BreakOutFloat2Components}
  n6 = AppendVector(A=Multiply(A=@n5.R, B=@n5.R), B=@n5.G)
  n7 = Add(A=@n6, B=Constant2Vector() {g=-0.25, r=0.5})
  n8 = Constant2Vector() {g=2, r=0.5}
  n9 = MaterialFunctionCall(
   Input=SphereMask(A=Multiply(A=@n7, B=@n8), B=Constant() {r=0.5}, Radius=Constant() {r=0.5}, Hardness=Constant() {r=-6})
   Input Low=Constant()
   Input High=Constant() {r=1}
  ) {materialFunction=&RemapValue_0-1}
  n10 = Add(A=ComponentMask(None=TextureCoordinate()) {r=true}) {constB=-0.25}
  n11 = OneMinus(None=Multiply(A=Multiply(A=@n9.Result, B=@n10)) {constB=10})
  n12 = VectorParameter() {parameterName=Param_1, defaultValue=<rgba 1 1 1 1>}
  n13 = VectorParameter() {defaultValue=<rgba 0.5 0.5 0.5 0>}
  n14 = Constant() {r=0.5}
  n15 = SphereMask(A=@n6, B=@n14, Radius=@n14, Hardness=Constant())
  n16 = Add(A=ComponentMask(None=@n6) {g=true}) {constB=-0.5}
  n17 = Sine(None=ComponentMask(None=Multiply(A=@n6) {constB=0.5}) {r=true})
  n18 = ShadingPathSwitch(Default=ParticleRandom(), Mobile=Constant()) {desc=Migrated from FeatureLevelSwitch}
  n19 = DynamicParameter() {paramNames [Random,Param2,Param3,Param4]}
  n20 = TextureSample(UVs=AppendVector(A=Divide(A=@n16, B=@n17), B=StaticSwitchParameter(True=@n18, False=@n19.Random) {parameterName=Use Particle Random})) {texture=/Game/Effects/Fort_Effects/Textures/Gate/S_Gate_Swrill}
  n21 = Add(A=Multiply(A=@n15), B=ShadingPathSwitch(Default=Multiply(A=@n5.R, B=@n20.G), Mobile=@n5.R) {desc=Migrated from FeatureLevelSwitch})
  MaterialExpressionOneMinus_1 = OneMinus(None=MaterialFunctionCall(XY=@n6) {materialFunction=&Swizzle}.YX)
  n22 = Custom!(uv=@MaterialExpressionOneMinus_1) {
   code="uv = uv * 2 - 1; \r\nuv.y = uv.y + 0.7;\r\n\r\nfloat3 result = float3(0, 0, 0); \r\nfloat dist = length(uv);       \r\n\r\nif (dist <= 0.35 * (1.5 + 3 * sin(atan2(uv.y, uv.x*2)))) \r\n{\r\n    result = float3(1, 0, 0);\r\n}\r\n\r\nreturn result;"
   outputType=CMOT_Float2
   showCode=true
   inputs [{
     inputName=uv
     input {expression=&self:MaterialExpressionOneMinus_1}
    }]
  }
  n23 = Multiply(A=@n22, B=@n15)
  n24 = Multiply(A=@n23) {constB=3}
  n25 = Multiply(
   A=LinearInterpolate(A=Saturate(None=@n11), Alpha=DynamicParameter() {paramNames [Random,Param2,Param3,Param4]}.Random)
   B=LinearInterpolate(
    A=VectorParameter() {parameterName=Param_2, defaultValue=<rgba 0.25 0.25 0.25 0>}.RGB
    B=LinearInterpolate(A=@n12.RGB, B=@n13.RGB, Alpha=@n21)
    Alpha=Saturate(None=@n24)
   )
  )
  LinearInterpolate(Alpha=OneMinus(None=@n25)) {constA=0.5}
  Multiply(A=ScalarParameter() {defaultValue=0.5, parameterName=SSS_Opacity}, B=OneMinus(None=@n23))
  Multiply(A=Subtract(A=TextureSample(UVs=@n6) {samplerType=SAMPLERTYPE_LinearColor, texture=/Game/Packages/Fortress_SharedMaterials/Textures/SphereNormalsUncomp}.RGB) {constB=0.5}) {constB=2}
  Multiply(A=Multiply(A=ParticleColor().RGB, B=@n25), B=ScalarParameter() {defaultValue=0.1, parameterName=SSS_Color})
 ]
}
