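# Control Rig for the Thorned Eden medium female body cosmetic, correcting deformation of shoulder roses, straps and gun.
# The shared DeformRig function runs first, then weighted parent constraints attach rose base bones to spine, clavicle and arm bones.
# Spherical pose readers on upper arms, thighs and upper spine then weight additional transform offsets and constraints as corrective poses.
# Variables toggle each reader's debug drawing by limb.
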ControlRigBlueprint /BRCosmetics/Characters/Player/Female/Medium/Bodies/F_MED_ThornedEden/Meshes/F_MED_ThornedEden_CtrlRig
graphs [
 RigVMFunctionLibrary {links [], nodes []}
 RigVMFunction_Sequence_12_1_ContainedGraph {nodes [
   n1 = RigVMFunctionEntryNode()
   n2 = &RigVMFunction_Sequence(ExecuteContext=@n1.ExecuteContext)
   n3 = &RigVMFunction_Sequence(ExecuteContext=@n2.B)
   RigVMFunctionReturnNode(A=@n2.A, B=@n3.A, C=@n3.B)
  ]}
 RigVMFunction_Sequence_15_1_ContainedGraph {nodes [
   n1 = RigVMFunctionEntryNode()
   n2 = &RigVMFunction_Sequence(ExecuteContext=@n1.ExecuteContext)
   n3 = &RigVMFunction_Sequence(ExecuteContext=@n2.B)
   RigVMFunctionReturnNode(A=@n2.A, B=@n3.A, C=@n3.B)
  ]}
 RigVMFunction_Sequence_2_ContainedGraph {nodes [
   n1 = RigVMFunctionEntryNode()
   RigVMFunction_Sequence = &RigVMFunction_Sequence(ExecuteContext=@n1.ExecuteContext)
   n2 = &RigVMFunction_Sequence(ExecuteContext=@RigVMFunction_Sequence.B)
   n3 = &RigVMFunction_Sequence(ExecuteContext=@n2.B)
   RigVMFunctionReturnNode(A=@RigVMFunction_Sequence.A, B=@n2.A, C=@n3.A, D=@n3.B)
  ]}
 RigVMFunction_Sequence_2_ContainedGraph {nodes [
   n1 = RigVMFunctionEntryNode()
   RigVMFunction_Sequence = &RigVMFunction_Sequence(ExecuteContext=@n1.ExecuteContext)
   n2 = &RigVMFunction_Sequence(ExecuteContext=@RigVMFunction_Sequence.B)
   n3 = &RigVMFunction_Sequence(ExecuteContext=@n2.B)
   n4 = &RigVMFunction_Sequence(ExecuteContext=@n3.B)
   RigVMFunctionReturnNode(A=@RigVMFunction_Sequence.A, B=@n2.A, C=@n3.A, D=@n4.A, E=@n4.B)
  ]}
 RigVMFunction_Sequence_3_1_ContainedGraph {nodes [
   n1 = RigVMFunctionEntryNode()
   n2 = &RigVMFunction_Sequence(ExecuteContext=@n1.ExecuteContext)
   n3 = &RigVMFunction_Sequence(ExecuteContext=@n2.B)
   n4 = &RigVMFunction_Sequence(ExecuteContext=@n3.B)
   RigVMFunctionReturnNode(A=@n2.A, B=@n3.A, C=@n4.A, D=@n4.B)
  ]}
 RigVMModel {nodes [
   RigUnit_BeginExecution = &RigUnit_BeginExecution()
   RigVMFunction_Sequence = RigVMAggregateNode(ExecuteContext=@RigUnit_BeginExecution.ExecutePin)
   n1 = &RigVMFunction_Sequence(ExecuteContext=@RigVMFunction_Sequence.C)
   RigVMFunction_Sequence_2 = RigVMAggregateNode(ExecuteContext=@n1.B)
   n2 = &RigVMFunction_Sequence(ExecuteContext=@RigVMFunction_Sequence_2.A)
   n3 = &RigUnit_SphericalPoseReader(
    ActiveRegionScaleFactors=(PositiveWidth=1,NegativeWidth=1,PositiveHeight=1,NegativeHeight=1)
    ActiveRegionSize=0.05
    Debug=(bDrawDebug=true,bDraw2D=False,bDrawLocalAxes=False,DebugScale=25,DebugSegments=20,DebugThickness=0.25)
    DriverAxis=(X=-1,Y=0,Z=0)
    DriverItem=(Type=Bone,Name="upperarm_r")
    FalloffRegionScaleFactors=(PositiveWidth=1,NegativeWidth=1,PositiveHeight=2,NegativeHeight=1)
    FalloffSize=0.3
    OptionalParentItem=(Type=Bone)
    RotationOffset=(X=-115.9962,Y=15.2218,Z=33.54)
    ExecutePin=@n2.B
    Debug.bDrawDebug=RigVMVariableNode() {variable=Degub_arm_r}.Value
   )
   > &RigUnit_ModifyTransforms(
    ItemToModify=((Item=(Type=Bone,Name="dyn_rose_02_r_bk_base"),Transform=(Rotation=(X=-0.0443,Y=-0.1217,Z=-0.2557,W=0.958),Translation=(X=-1.1029,Y=-1.4489,Z=1.3742),Scale3D=(X=1,Y=1,Z=1))))
    WeightMinimum=0
    Weight=@n3.OutputParam
   )
   > &RigUnit_ModifyTransforms(
    ItemToModify=((Item=(Type=Bone,Name="dyn_rose_01_r_bk_down_base"),Transform=(Rotation=(X=0,Y=0,Z=0,W=1),Translation=(X=0.1147,Y=1.2964,Z=2.8469),Scale3D=(X=1,Y=1,Z=1))))
    WeightMinimum=0
    Weight=@n3.OutputParam
   )
   > &RigUnit_ModifyTransforms(
    ItemToModify=((Item=(Type=Bone,Name="dyn_rose_r_side_base"),Transform=(Rotation=(X=-0.1383,Y=0.1883,Z=-0.0527,W=-0.9709),Translation=(X=-0.5409,Y=2.2425,Z=-0.327),Scale3D=(X=1,Y=1,Z=1))))
    WeightMinimum=0
    Weight=@n3.OutputParam
   )
   n4 = &RigUnit_SphericalPoseReader(
    ActiveRegionScaleFactors=(PositiveWidth=1,NegativeWidth=2,PositiveHeight=1,NegativeHeight=1)
    ActiveRegionSize=0.08
    Debug=(bDrawDebug=true,bDraw2D=False,bDrawLocalAxes=False,DebugScale=25,DebugSegments=20,DebugThickness=0.25)
    DriverAxis=(X=-1,Y=0,Z=0)
    DriverItem=(Type=Bone,Name="upperarm_r")
    FalloffRegionScaleFactors=(PositiveWidth=1,NegativeWidth=1,PositiveHeight=1,NegativeHeight=1)
    FalloffSize=0.4
    OptionalParentItem=(Type=Bone)
    RotationOffset=(X=-116.5844,Y=8.5575,Z=14.8418)
    Debug.bDrawDebug=RigVMVariableNode() {variable=Degub_arm_r}.Value
    ExecutePin=@n2.A
   )
   > &RigUnit_ParentConstraint(
    AdvancedSettings=(InterpolationType=Average,RotationOrderForFilter=XZY)
    Child=(Type=Bone,Name="dyn_rose_02_r_fr_base")
    Parents=((Item=(Type=Bone,Name="upperarm_r"),Weight=0.1775),(Item=(Type=Bone,Name="upperarm_twist_02_r"),Weight=0.2701),(Item=(Type=Bone,Name="upperarm_twist_01_r"),Weight=0.5524))
    bMaintainOffset=true
    Weight=@n4.OutputParam
   )
   > &RigUnit_ModifyTransforms(
    ItemToModify=((Item=(Type=Bone,Name="dyn_rose_02_r_fr_base"),Transform=(Rotation=(X=0.0812,Y=-0.5331,Z=-0.1511,W=0.8285),Translation=(X=-2.8544,Y=9.2269,Z=9.0385),Scale3D=(X=1,Y=1,Z=1))))
    WeightMinimum=0
    Weight=@n4.OutputParam
   )
   n5 = &RigUnit_SphericalPoseReader(
    ActiveRegionScaleFactors=(PositiveWidth=1,NegativeWidth=1,PositiveHeight=1,NegativeHeight=1)
    ActiveRegionSize=0.02
    Debug=(bDrawDebug=true,bDraw2D=False,bDrawLocalAxes=False,DebugScale=25,DebugSegments=20,DebugThickness=0.25)
    DriverAxis=(X=0,Y=0,Z=-1)
    DriverItem=(Type=Bone,Name="upperarm_r")
    FalloffRegionScaleFactors=(PositiveWidth=1,NegativeWidth=0.5,PositiveHeight=1,NegativeHeight=1)
    FalloffSize=0.3
    OptionalParentItem=(Type=Bone)
    RotationOffset=(X=-147.381,Y=0,Z=-41.8939)
    Debug.bDrawDebug=RigVMVariableNode() {variable=Degub_arm_r}.Value
    ExecutePin=@RigVMFunction_Sequence_2.B
   )
   > &RigUnit_ModifyTransforms(
    ItemToModify=((Item=(Type=Bone,Name="dyn_rose_02_r_bk_base"),Transform=(Rotation=(X=-0.0136,Y=-0.0069,Z=-0.2591,W=0.9657),Translation=(X=-1.3245,Y=0.2319,Z=1.5518),Scale3D=(X=1,Y=1,Z=1))))
    WeightMinimum=0
    Weight=@n5.OutputParam
   )
   > &RigUnit_ModifyTransforms(
    ItemToModify=((Item=(Type=Bone,Name="dyn_rose_r_side_base"),Transform=(Rotation=(X=-0.2625,Y=-0.1038,Z=-0.2648,W=0.9221),Translation=(X=-0.4067,Y=-1.8462,Z=1.5864),Scale3D=(X=1,Y=1,Z=1))))
    WeightMinimum=0
    Weight=@n5.OutputParam
   )
   n6 = &RigUnit_SphericalPoseReader(
    ActiveRegionScaleFactors=(PositiveWidth=1,NegativeWidth=1,PositiveHeight=1,NegativeHeight=1)
    ActiveRegionSize=0.03
    Debug=(bDrawDebug=true,bDraw2D=False,bDrawLocalAxes=False,DebugScale=25,DebugSegments=20,DebugThickness=0.25)
    DriverAxis=(X=-1,Y=0,Z=0)
    DriverItem=(Type=Bone,Name="upperarm_r")
    FalloffRegionScaleFactors=(PositiveWidth=1,NegativeWidth=1,PositiveHeight=1,NegativeHeight=1)
    FalloffSize=0.2
    OptionalParentItem=(Type=Bone)
    RotationOffset=(X=-27.0558,Y=-52.3519,Z=-70)
    Debug.bDrawDebug=RigVMVariableNode() {variable=Degub_arm_r}.Value
    ExecutePin=@RigVMFunction_Sequence_2.D
   )
   > &RigUnit_ModifyTransforms(
    ItemToModify=((Item=(Type=Bone,Name="dyn_rose_r_side_base"),Transform=(Rotation=(X=-0.0458,Y=-0.0626,Z=0.0075,W=0.997),Translation=(X=0,Y=0,Z=0),Scale3D=(X=1,Y=1,Z=1))))
    WeightMinimum=0
    Weight=@n6.OutputParam
   )
   > &RigUnit_ModifyTransforms(
    ItemToModify=((Item=(Type=Bone,Name="dyn_rose_02_r_fr_base"),Transform=(Rotation=(X=0,Y=-0.1078,Z=0,W=0.9942),Translation=(X=-0.0364,Y=4.3458,Z=2.551),Scale3D=(X=1,Y=1,Z=1))))
    WeightMinimum=0
    Weight=@n6.OutputParam
   )
   n7 = &RigUnit_SphericalPoseReader(
    ActiveRegionScaleFactors=(PositiveWidth=1,NegativeWidth=1,PositiveHeight=1,NegativeHeight=1)
    ActiveRegionSize=0.03
    Debug=(bDrawDebug=true,bDraw2D=False,bDrawLocalAxes=False,DebugScale=25,DebugSegments=20,DebugThickness=0.25)
    DriverAxis=(X=-1,Y=0,Z=0)
    DriverItem=(Type=Bone,Name="upperarm_r")
    FalloffRegionScaleFactors=(PositiveWidth=1,NegativeWidth=1,PositiveHeight=1,NegativeHeight=1)
    FalloffSize=0.2
    OptionalParentItem=(Type=Bone)
    RotationOffset=(X=-54.1456,Y=-38.2666,Z=-112.4219)
    Debug.bDrawDebug=RigVMVariableNode() {variable=Degub_arm_r}.Value
    ExecutePin=@RigVMFunction_Sequence_2.E
   )
   > &RigUnit_ModifyTransforms(
    ItemToModify=((Item=(Type=Bone,Name="dyn_rose_r_side_base"),Transform=(Rotation=(X=-0.0464,Y=0.0816,Z=0.0008,W=0.9956),Translation=(X=-0.1281,Y=-0.1288,Z=-1.3766),Scale3D=(X=1,Y=1,Z=1))))
    WeightMinimum=0
    Weight=@n7.OutputParam
   )
   > &RigUnit_ModifyTransforms(
    ItemToModify=((Item=(Type=Bone,Name="dyn_rose_02_r_fr_base"),Transform=(Rotation=(X=-0.0055,Y=-0.0261,Z=-0.0052,W=0.9996),Translation=(X=-0.528,Y=1.0604,Z=-1.3328),Scale3D=(X=1,Y=1,Z=1))))
    WeightMinimum=0
    Weight=@n7.OutputParam
   )
   n8 = &RigUnit_SphericalPoseReader(
    ActiveRegionScaleFactors=(PositiveWidth=1,NegativeWidth=1,PositiveHeight=1,NegativeHeight=1)
    ActiveRegionSize=0.02
    Debug=(bDrawDebug=true,bDraw2D=False,bDrawLocalAxes=False,DebugScale=25,DebugSegments=20,DebugThickness=0.25)
    DriverAxis=(X=-1,Y=0,Z=0)
    DriverItem=(Type=Bone,Name="upperarm_r")
    FalloffRegionScaleFactors=(PositiveWidth=1,NegativeWidth=1,PositiveHeight=1,NegativeHeight=1)
    FalloffSize=0.3
    OptionalParentItem=(Type=Bone)
    RotationOffset=(X=-30.0493,Y=-10.3001,Z=-70)
    Debug.bDrawDebug=RigVMVariableNode() {variable=Degub_arm_r}.Value
    ExecutePin=@RigVMFunction_Sequence_2.C
   )
   > &RigUnit_ModifyTransforms(
    ItemToModify=((Item=(Type=Bone,Name="dyn_rose_02_r_bk_base"),Transform=(Rotation=(X=-0.0205,Y=-0.042,Z=-0.1716,W=0.9841),Translation=(X=-1.3245,Y=0.2319,Z=1.5518),Scale3D=(X=1,Y=1,Z=1))))
    WeightMinimum=0
    Weight=@n8.OutputParam
   )
   > &RigUnit_ModifyTransforms(
    ItemToModify=((Item=(Type=Bone,Name="dyn_rose_02_r_fr_base"),Transform=(Rotation=(X=0.03,Y=-0.2825,Z=-0.1735,W=0.943),Translation=(X=0,Y=0,Z=1.9035),Scale3D=(X=1,Y=1,Z=1))))
    WeightMinimum=0
    Weight=@n8.OutputParam
   )
   > &RigUnit_ModifyTransforms(
    ItemToModify=((Item=(Type=Bone,Name="dyn_rose_r_side_base"),Transform=(Rotation=(X=0.0818,Y=-0.4692,Z=0.0261,W=0.8789),Translation=(X=-1.2395,Y=1.0667,Z=3.5019),Scale3D=(X=1,Y=1,Z=1))))
    WeightMinimum=0
    Weight=@n8.OutputParam
   )
   RigVMFunction_Sequence_15 = RigVMAggregateNode(ExecuteContext=@n1.A)
   n9 = &RigUnit_SphericalPoseReader(
    ActiveRegionScaleFactors=(PositiveWidth=1,NegativeWidth=1,PositiveHeight=1,NegativeHeight=1)
    ActiveRegionSize=0.1
    Debug=(bDrawDebug=true,bDraw2D=False,bDrawLocalAxes=False,DebugScale=25,DebugSegments=20,DebugThickness=0.25)
    DriverAxis=(X=1,Y=0,Z=0)
    DriverItem=(Type=Bone,Name="upperarm_l")
    FalloffRegionScaleFactors=(PositiveWidth=1,NegativeWidth=1,PositiveHeight=1,NegativeHeight=1)
    FalloffSize=0.25
    OptionalParentItem=(Type=Bone)
    RotationOffset=(X=50,Y=1.5913,Z=-35)
    ExecutePin=@RigVMFunction_Sequence_15.A
    Debug.bDrawDebug=RigVMVariableNode() {variable=Degub_arm_l}.Value
   )
   > &RigUnit_ParentConstraint(
    AdvancedSettings=(RotationOrderForFilter=XZY)
    Child=(Type=Bone,Name="dyn_rose_02_l_fr_base")
    Parents=((Item=(Type=Bone,Name="upperarm_l"),Weight=0.2168),(Item=(Type=Bone,Name="upperarm_twist_01_l"),Weight=0.2816),(Item=(Type=Bone,Name="upperarm_twist_02_l"),Weight=0.5016))
    Weight=@n9.OutputParam
   )
   > &RigUnit_ModifyTransforms(
    ItemToModify=((Item=(Type=Bone,Name="dyn_rose_02_l_fr_base"),Transform=(Rotation=(X=-0.0827,Y=-0.3393,Z=0.1869,W=0.9182),Translation=(X=-3.8426,Y=7.2598,Z=-5.1626),Scale3D=(X=1,Y=1,Z=1))))
    WeightMinimum=0
    Weight=@n9.OutputParam
   )
   n10 = &RigUnit_SphericalPoseReader(
    ActiveRegionScaleFactors=(PositiveWidth=1,NegativeWidth=1,PositiveHeight=1,NegativeHeight=1)
    ActiveRegionSize=0.1
    Debug=(bDrawDebug=true,bDraw2D=False,bDrawLocalAxes=False,DebugScale=25,DebugSegments=20,DebugThickness=0.25)
    DriverAxis=(X=1,Y=0,Z=0)
    DriverItem=(Type=Bone,Name="upperarm_l")
    FalloffRegionScaleFactors=(PositiveWidth=1,NegativeWidth=1,PositiveHeight=1,NegativeHeight=1)
    FalloffSize=0.2
    OptionalParentItem=(Type=Bone)
    RotationOffset=(X=85,Y=1.5913,Z=-35)
    ExecutePin=@RigVMFunction_Sequence_15.B
    Debug.bDrawDebug=RigVMVariableNode() {variable=Degub_arm_l}.Value
   )
   > &RigUnit_ModifyTransforms(
    ItemToModify=((Item=(Type=Bone,Name="dyn_rose_02_l_fr_base"),Transform=(Rotation=(X=0,Y=0,Z=0.1125,W=-0.9937),Translation=(X=0.3626,Y=0,Z=-0.9901),Scale3D=(X=1,Y=1,Z=1))))
    WeightMinimum=0
    Weight=@n10.OutputParam
   )
   n11 = &RigUnit_SphericalPoseReader(
    ActiveRegionScaleFactors=(PositiveWidth=1,NegativeWidth=1,PositiveHeight=1,NegativeHeight=1)
    ActiveRegionSize=0.05
    Debug=(bDrawDebug=true,bDraw2D=False,bDrawLocalAxes=False,DebugScale=25,DebugSegments=20,DebugThickness=0.25)
    DriverAxis=(X=1,Y=0,Z=0)
    DriverItem=(Type=Bone,Name="upperarm_l")
    FalloffRegionScaleFactors=(PositiveWidth=1,NegativeWidth=1,PositiveHeight=1,NegativeHeight=1)
    FalloffSize=0.2
    OptionalParentItem=(Type=Bone)
    RotationOffset=(X=115.1186,Y=1.1915,Z=-38.6207)
    Debug.bDrawDebug=RigVMVariableNode() {variable=Degub_arm_l}.Value
    ExecutePin=@RigVMFunction_Sequence_15.C
   )
   > &RigUnit_ModifyTransforms(
    ItemToModify=((Item=(Type=Bone,Name="dyn_rose_l_side_base"),Transform=(Rotation=(X=0,Y=0.1593,Z=0,W=0.9872),Translation=(X=-0.7725,Y=0,Z=-2.4992),Scale3D=(X=1,Y=1,Z=1))))
    WeightMinimum=0
    Weight=@n11.OutputParam
   )
   > &RigUnit_ModifyTransforms(
    ItemToModify=((Item=(Type=Bone,Name="dyn_rose_02_l_bk_base"),Transform=(Rotation=(X=0,Y=0,Z=0,W=1),Translation=(X=0,Y=-1.5456,Z=-1.8657),Scale3D=(X=1,Y=1,Z=1))))
    WeightMinimum=0
    Weight=@n11.OutputParam
   )
   > &RigUnit_ModifyTransforms(
    ItemToModify=((Item=(Type=Bone,Name="dyn_rose_02_l_fr_base"),Transform=(Rotation=(X=0,Y=0,Z=0,W=1),Translation=(X=0,Y=0.7359,Z=-2.3453),Scale3D=(X=1,Y=1,Z=1))))
    WeightMinimum=0
    Weight=@n11.OutputParam
   )
   n12 = &RigVMFunction_Sequence(ExecuteContext=@RigVMFunction_Sequence.B)
   &RigUnit_ParentConstraint(
    AdvancedSettings=(RotationOrderForFilter=XZY)
    Child=(Type=Bone,Name="dyn_rose_l_side_base")
    Parents="((Item=(Type=Bone,Name=\"spine_05\"),Weight=0.0271),(Item=(Type=Bone,Name=\"deform_upperarm_fr_l\"),Weight=0.0429),(Item=(Type=Bone,Name=\"deform_upperarm_bk_l\"),Weight=0.059),(Item=(Type=Bone,Name=\"deform_upperarm_up_l\"),Weight=0.1383),(Item=(Type=Bone,Name=\"c ...(truncated)"
    Weight=1
    ExecutePin=@n12.A
   )
   > &RigUnit_ParentConstraint(
    AdvancedSettings=(RotationOrderForFilter=XZY)
    Child=(Type=Bone,Name="dyn_rose_02_l_fr_base")
    Parents="((Item=(Type=Bone,Name=\"upperarm_twist_01_l\"),Weight=0.0274),(Item=(Type=Bone,Name=\"deform_upperarm_up_l\"),Weight=0.0523),(Item=(Type=Bone,Name=\"spine_04\"),Weight=0.0841),(Item=(Type=Bone,Name=\"spine_05\"),Weight=0.1008),(Item=(Type=Bone,Name=\"deform_uppera ...(truncated)"
    Weight=1
   )
   > &RigUnit_ParentConstraint(
    AdvancedSettings=(RotationOrderForFilter=XZY)
    Child=(Type=Bone,Name="dyn_rose_01_l_fr_base")
    Parents="((Item=(Type=Bone,Name=\"upperarm_l\"),Weight=0.0269),(Item=(Type=Bone,Name=\"deform_pec_aim_l\"),Weight=0.0287),(Item=(Type=Bone,Name=\"clavicle_l\"),Weight=0.1016),(Item=(Type=Bone,Name=\"spine_04\"),Weight=0.2401),(Item=(Type=Bone,Name=\"spine_05\"),Weight=0.6027 ...(truncated)"
    Weight=1
   )
   > &RigUnit_ParentConstraint(
    AdvancedSettings=(RotationOrderForFilter=XZY)
    Child=(Type=Bone,Name="dyn_rose_01_l_fr_down_base")
    Parents=((Item=(Type=Bone,Name="clavicle_l"),Weight=0.0508),(Item=(Type=Bone,Name="deform_pec_aim_l"),Weight=0.0633),(Item=(Type=Bone,Name="spine_04"),Weight=0.4404),(Item=(Type=Bone,Name="spine_05"),Weight=0.4454))
    Weight=1
   )
   > &RigUnit_ParentConstraint(
    AdvancedSettings=(RotationOrderForFilter=XZY)
    Child=(Type=Bone,Name="dyn_rose_center_fr_base")
    Parents=((Item=(Type=Bone,Name="spine_04"),Weight=0.1911),(Item=(Type=Bone,Name="spine_05"),Weight=0.8089))
    Weight=1
   )
   > &RigUnit_ParentConstraint(
    AdvancedSettings=(RotationOrderForFilter=XZY)
    Child=(Type=Bone,Name="dyn_rose_center_fr_down_base")
    Parents=((Item=(Type=Bone,Name="spine_05"),Weight=0.4747),(Item=(Type=Bone,Name="spine_04"),Weight=0.5253))
    Weight=1
   )
   > &RigUnit_ParentConstraint(
    AdvancedSettings=(RotationOrderForFilter=XZY)
    Child=(Type=Bone,Name="dyn_rose_01_r_fr_base")
    Parents="((Item=(Type=Bone,Name=\"deform_pec_aim_r\"),Weight=0.0235),(Item=(Type=Bone,Name=\"deform_upperarm_fr_r\"),Weight=0.0343),(Item=(Type=Bone,Name=\"upperarm_r\"),Weight=0.0553),(Item=(Type=Bone,Name=\"clavicle_r\"),Weight=0.1657),(Item=(Type=Bone,Name=\"spine_04\"),W ...(truncated)"
    Weight=1
   )
   > &RigUnit_ParentConstraint(
    AdvancedSettings=(RotationOrderForFilter=XZY)
    Child=(Type=Bone,Name="dyn_rose_01_r_fr_down_base")
    Parents=((Item=(Type=Bone,Name="clavicle_r"),Weight=0.0526),(Item=(Type=Bone,Name="deform_pec_aim_r"),Weight=0.0635),(Item=(Type=Bone,Name="spine_04"),Weight=0.437),(Item=(Type=Bone,Name="spine_05"),Weight=0.4469))
    Weight=1
   )
   > &RigUnit_ParentConstraint(
    AdvancedSettings=(RotationOrderForFilter=XZY)
    Child=(Type=Bone,Name="dyn_rose_02_r_fr_base")
    Parents="((Item=(Type=Bone,Name=\"deform_upperarm_up_r\"),Weight=0.0473),(Item=(Type=Bone,Name=\"upperarm_twist_01_r\"),Weight=0.0554),(Item=(Type=Bone,Name=\"spine_04\"),Weight=0.0711),(Item=(Type=Bone,Name=\"spine_05\"),Weight=0.0802),(Item=(Type=Bone,Name=\"clavicle_r\"), ...(truncated)"
    Weight=1
   )
   > &RigUnit_ParentConstraint(
    AdvancedSettings=(RotationOrderForFilter=XZY)
    Child=(Type=Bone,Name="dyn_rose_r_side_base")
    Parents="((Item=(Type=Bone,Name=\"deform_upperarm_fr_r\"),Weight=0.0394),(Item=(Type=Bone,Name=\"deform_upperarm_bk_r\"),Weight=0.0483),(Item=(Type=Bone,Name=\"clavicle_r\"),Weight=0.1209),(Item=(Type=Bone,Name=\"deform_upperarm_up_r\"),Weight=0.1243),(Item=(Type=Bone,Name ...(truncated)"
    Weight=1
   )
   &RigUnit_ParentConstraint(
    AdvancedSettings=(RotationOrderForFilter=XZY)
    Child=(Type=Bone,Name="dyn_rose_02_l_bk_base")
    Parents="((Item=(Type=Bone,Name=\"deform_pec_l\"),Weight=0.0215),(Item=(Type=Bone,Name=\"spine_04\"),Weight=0.027),(Item=(Type=Bone,Name=\"deform_upperarm_up_l\"),Weight=0.1203),(Item=(Type=Bone,Name=\"clavicle_l\"),Weight=0.1438),(Item=(Type=Bone,Name=\"deform_upperarm_bk_ ...(truncated)"
    Weight=1
    ExecutePin=@n12.B
   )
   > &RigUnit_ParentConstraint(
    AdvancedSettings=(RotationOrderForFilter=XZY)
    Child=(Type=Bone,Name="dyn_rose_01_l_bk_base")
    Parents="((Item=(Type=Bone,Name=\"spine_04\"),Weight=0.0283),(Item=(Type=Bone,Name=\"upperarm_l\"),Weight=0.0393),(Item=(Type=Bone,Name=\"deform_upperarm_bk_l\"),Weight=0.0556),(Item=(Type=Bone,Name=\"clavicle_l\"),Weight=0.1538),(Item=(Type=Bone,Name=\"spine_05\"),Weight=0. ...(truncated)"
    Weight=1
   )
   > &RigUnit_ParentConstraint(
    AdvancedSettings=(RotationOrderForFilter=XZY)
    Child=(Type=Bone,Name="dyn_rose_01_l_bk_down_base")
    Parents="((Item=(Type=Bone,Name=\"spine_03\"),Weight=0.0214),(Item=(Type=Bone,Name=\"clavicle_l\"),Weight=0.0377),(Item=(Type=Bone,Name=\"deform_upperarm_bk_l\"),Weight=0.0582),(Item=(Type=Bone,Name=\"upperarm_l\"),Weight=0.1254),(Item=(Type=Bone,Name=\"spine_04\"),Weight=0. ...(truncated)"
    Weight=1
   )
   > &RigUnit_ParentConstraint(
    AdvancedSettings=(RotationOrderForFilter=XZY)
    Child=(Type=Bone,Name="dyn_rose_center_bk_down_base")
    Parents=((Item=(Type=Bone,Name="spine_04"),Weight=0.3117),(Item=(Type=Bone,Name="spine_05"),Weight=0.6883))
    Weight=1
   )
   > &RigUnit_ParentConstraint(
    AdvancedSettings=(RotationOrderForFilter=XZY)
    Child=(Type=Bone,Name="dyn_rose_01_r_bk_base")
    Parents="((Item=(Type=Bone,Name=\"upperarm_r\"),Weight=0.0254),(Item=(Type=Bone,Name=\"deform_upperarm_bk_r\"),Weight=0.0751),(Item=(Type=Bone,Name=\"spine_04\"),Weight=0.089),(Item=(Type=Bone,Name=\"clavicle_r\"),Weight=0.1049),(Item=(Type=Bone,Name=\"spine_05\"),Weight=0.7 ...(truncated)"
    Weight=1
   )
   > &RigUnit_ParentConstraint(
    AdvancedSettings=(RotationOrderForFilter=XZY)
    Child=(Type=Bone,Name="dyn_rose_01_r_bk_down_base")
    Parents="((Item=(Type=Bone,Name=\"clavicle_r\"),Weight=0.0644),(Item=(Type=Bone,Name=\"deform_upperarm_bk_r\"),Weight=0.078),(Item=(Type=Bone,Name=\"upperarm_r\"),Weight=0.1271),(Item=(Type=Bone,Name=\"spine_04\"),Weight=0.1709),(Item=(Type=Bone,Name=\"spine_05\"),Weight=0.5 ...(truncated)"
    Weight=1
   )
   > &RigUnit_ParentConstraint(
    AdvancedSettings=(RotationOrderForFilter=XZY)
    Child=(Type=Bone,Name="dyn_rose_02_r_bk_base")
    Parents="((Item=(Type=Bone,Name=\"deform_pec_r\"),Weight=0.0214),(Item=(Type=Bone,Name=\"spine_04\"),Weight=0.0223),(Item=(Type=Bone,Name=\"deform_upperarm_up_r\"),Weight=0.1288),(Item=(Type=Bone,Name=\"clavicle_r\"),Weight=0.1505),(Item=(Type=Bone,Name=\"deform_upperarm_bk ...(truncated)"
    Weight=1
   )
   RigVMFunction_Sequence_3_1 = RigVMAggregateNode(ExecuteContext=@RigVMFunction_Sequence.D)
   n13 = &RigVMFunction_Sequence(ExecuteContext=@RigVMFunction_Sequence_3_1.A)
   n14 = &RigUnit_SphericalPoseReader(
    ActiveRegionScaleFactors=(PositiveWidth=2,NegativeWidth=2,PositiveHeight=2,NegativeHeight=1)
    ActiveRegionSize=0.05
    Debug=(bDrawDebug=true,bDraw2D=False,bDrawLocalAxes=False,DebugScale=25,DebugSegments=20,DebugThickness=0.25)
    DriverAxis=(X=1,Y=0,Z=0)
    DriverItem=(Type=Bone,Name="upperarm_l")
    FalloffRegionScaleFactors=(PositiveWidth=1.5,NegativeWidth=1.5,PositiveHeight=3,NegativeHeight=1)
    FalloffSize=0.2
    OptionalParentItem=(Type=Bone)
    RotationOffset=(X=52.3073,Y=-15.3455,Z=-80)
    Debug.bDrawDebug=RigVMVariableNode() {variable=Degub_arm_l}.Value
    ExecutePin=@n13.B
   )
   > &RigUnit_ModifyTransforms(
    ItemToModify=((Item=(Type=Bone,Name="dyn_rose_l_side_base"),Transform=(Rotation=(X=-0.0057,Y=-0.0432,Z=-0.1304,W=0.9905),Translation=(X=1,Y=5,Z=0),Scale3D=(X=1,Y=1,Z=1))))
    WeightMinimum=0
    Weight=@n14.OutputParam
   )
   > &RigUnit_ModifyTransforms(
    ItemToModify=((Item=(Type=Bone,Name="dyn_rose_02_l_fr_base"),Transform=(Rotation=(X=0.0302,Y=-0.171,Z=0.171,W=0.9698),Translation=(X=-1,Y=2.3342,Z=-1),Scale3D=(X=1,Y=1,Z=1))))
    WeightMinimum=0
    Weight=@n14.OutputParam
   )
   > &RigUnit_ModifyTransforms(
    ItemToModify=((Item=(Type=Bone,Name="dyn_rose_01_l_fr_base"),Transform=(Rotation=(X=-0.0302,Y=0.171,Z=0.171,W=0.9698),Translation=(X=-1,Y=-1,Z=-4),Scale3D=(X=1,Y=1,Z=1))))
    WeightMinimum=0
    Weight=@n14.OutputParam
   )
   > &RigUnit_ModifyTransforms(
    ItemToModify=((Item=(Type=Bone,Name="dyn_rose_01_l_bk_down_base"),Transform=(Rotation=(X=0,Y=0,Z=0,W=1),Translation=(X=0,Y=2.3436,Z=0.0657),Scale3D=(X=1,Y=1,Z=1))))
    WeightMinimum=0
    Weight=@n14.OutputParam
   )
   n15 = &RigUnit_SphericalPoseReader(
    ActiveRegionScaleFactors=(PositiveWidth=2,NegativeWidth=2,PositiveHeight=2,NegativeHeight=1)
    ActiveRegionSize=0.05
    Debug=(bDrawDebug=true,bDraw2D=False,bDrawLocalAxes=False,DebugScale=25,DebugSegments=20,DebugThickness=0.25)
    DriverAxis=(X=-1,Y=0,Z=0)
    DriverItem=(Type=Bone,Name="upperarm_r")
    FalloffRegionScaleFactors=(PositiveWidth=1.5,NegativeWidth=1.5,PositiveHeight=3,NegativeHeight=1)
    FalloffSize=0.2
    OptionalParentItem=(Type=Bone)
    RotationOffset=(X=-127.6927,Y=-15.3455,Z=-80)
    Debug.bDrawDebug=RigVMVariableNode() {variable=Degub_arm_r}.Value
    ExecutePin=@n13.A
   )
   > &RigUnit_ModifyTransforms(
    ItemToModify=((Item=(Type=Bone,Name="dyn_rose_r_side_base"),Transform=(Rotation=(X=0.0046,Y=0.0434,Z=-0.1044,W=0.9936),Translation=(X=0.5,Y=5,Z=0),Scale3D=(X=1,Y=1,Z=1))))
    WeightMinimum=0
    Weight=@n15.OutputParam
   )
   > &RigUnit_ModifyTransforms(
    ItemToModify=((Item=(Type=Bone,Name="dyn_rose_02_r_fr_base"),Transform=(Rotation=(X=-0.0302,Y=0.171,Z=0.171,W=0.9698),Translation=(X=-1,Y=2.3342,Z=1),Scale3D=(X=1,Y=1,Z=1))))
    WeightMinimum=0
    Weight=@n15.OutputParam
   )
   > &RigUnit_ModifyTransforms(
    ItemToModify=((Item=(Type=Bone,Name="dyn_rose_01_r_fr_base"),Transform=(Rotation=(X=0.0449,Y=-0.1677,Z=0.2549,W=0.9513),Translation=(X=-2,Y=-1,Z=4),Scale3D=(X=1,Y=1,Z=1))))
    WeightMinimum=0
    Weight=@n15.OutputParam
   )
   > &RigUnit_ModifyTransforms(
    ItemToModify=((Item=(Type=Bone,Name="dyn_rose_01_r_bk_down_base"),Transform=(Rotation=(X=0,Y=0,Z=0,W=1),Translation=(X=0.0001,Y=1.9617,Z=0),Scale3D=(X=1,Y=1,Z=1))))
    WeightMinimum=0
    Weight=@n15.OutputParam
   )
   > &RigUnit_ModifyTransforms(
    ItemToModify=((Item=(Type=Bone,Name="dyn_rose_02_r_bk_base"),Transform=(Rotation=(X=0,Y=0,Z=0,W=1),Translation=(X=-0.5,Y=2,Z=-1),Scale3D=(X=1,Y=1,Z=1))))
    WeightMinimum=0
    Weight=@n15.OutputParam
   )
   RigVMFunction_Sequence_12 = RigVMAggregateNode(ExecuteContext=@RigVMFunction_Sequence_3_1.B)
   n16 = &RigUnit_SphericalPoseReader(
    ActiveRegionScaleFactors=(PositiveWidth=3.5,NegativeWidth=0.5,PositiveHeight=3,NegativeHeight=3)
    ActiveRegionSize=0.1
    Debug=(bDrawDebug=True,bDraw2D=false,DebugScale=25,DebugSegments=20,DebugThickness=0.25)
    DriverAxis=(X=1,Y=0,Z=0)
    DriverItem=(Type=Bone,Name="thigh_r")
    FalloffRegionScaleFactors=(PositiveWidth=1,NegativeWidth=1,PositiveHeight=1,NegativeHeight=1)
    FalloffSize=0.45
    FlipHeightScaling=false
    FlipWidthScaling=true
    OptionalParentItem=(Type=Bone)
    RotationOffset=(X=-90,Y=0,Z=0)
    ExecutePin=@RigVMFunction_Sequence_12.A
    Debug.bDrawDebug=RigVMVariableNode() {variable=Degub_thigh_r}.Value
   )
   > &RigUnit_ModifyTransforms(
    ItemToModify=((Item=(Type=Bone,Name="extra_strap_r_bk"),Transform=(Rotation=(X=-0.0436,Y=0,Z=0,W=0.999),Translation=(X=0,Y=0,Z=0),Scale3D=(X=1,Y=1,Z=1))))
    WeightMinimum=0
    Weight=@n16.OutputParam
   )
   > &RigUnit_ModifyTransforms(
    ItemToModify=((Item=(Type=Bone,Name="extra_strap_r_fr"),Transform=(Rotation=(X=0,Y=0,Z=0.0436,W=0.999),Translation=(X=0,Y=-3,Z=0),Scale3D=(X=1,Y=1,Z=1))))
    WeightMinimum=0
    Weight=@n16.OutputParam
   )
   > &RigUnit_ModifyTransforms(
    ItemToModify=((Item=(Type=Bone,Name="extra_strap_center_fr"),Transform=(Rotation=(X=0,Y=0,Z=0,W=1),Translation=(X=1.3,Y=-3,Z=0),Scale3D=(X=1,Y=1,Z=1))))
    WeightMinimum=0
    Weight=@n16.OutputParam
   )
   n17 = &RigUnit_SphericalPoseReader(
    ActiveRegionScaleFactors=(PositiveWidth=3.5,NegativeWidth=0.5,PositiveHeight=3,NegativeHeight=3)
    ActiveRegionSize=0.1
    Debug=(bDrawDebug=True,DebugScale=25,DebugSegments=20,DebugThickness=0.25)
    DriverAxis=(X=-1,Y=0,Z=0)
    DriverItem=(Type=Bone,Name="thigh_l")
    FalloffRegionScaleFactors=(PositiveWidth=1,NegativeWidth=1,PositiveHeight=1,NegativeHeight=1)
    FalloffSize=0.45
    OptionalParentItem=(Type=Bone)
    RotationOffset=(X=90,Y=0,Z=0)
    ExecutePin=@RigVMFunction_Sequence_12.B
    Debug.bDrawDebug=RigVMVariableNode() {variable=Degub_thigh_l}.Value
   )
   > &RigUnit_ModifyTransforms(
    ItemToModify=((Item=(Type=Bone,Name="extra_strap_l_fr"),Transform=(Rotation=(X=0,Y=0,Z=0,W=-1),Translation=(X=-0.0002,Y=-7.9253,Z=1.1564),Scale3D=(X=1,Y=1,Z=1))))
    WeightMinimum=0
    Weight=@n17.OutputParam
   )
   > &RigUnit_ModifyTransforms(
    ItemToModify=((Item=(Type=Bone,Name="extra_strap_l_bk"),Transform=(Rotation=(X=0.0772,Y=0.1161,Z=0.0144,W=0.9901),Translation=(X=0,Y=-2.8535,Z=-4.5293),Scale3D=(X=1,Y=1,Z=1))))
    WeightMinimum=0
    Weight=@n17.OutputParam
   )
   > &RigUnit_ModifyTransforms(
    ItemToModify=((Item=(Type=Bone,Name="extra_strap_center_fr"),Transform=(Rotation=(X=-0.0905,Y=-0.0015,Z=0.0103,W=0.9958),Translation=(X=0.2075,Y=-2.2566,Z=0.2277),Scale3D=(X=1,Y=1,Z=1))))
    WeightMinimum=0
    Weight=@n17.OutputParam
   )
   > &RigUnit_ParentConstraint(
    AdvancedSettings=(RotationOrderForFilter=XZY)
    Child=(Type=Bone,Name="dyn_gun_base")
    Filter=(TranslationFilter=(bX=false,bY=True,bZ=True),RotationFilter=(bX=false,bY=false,bZ=false),ScaleFilter=(bX=True,bY=True,bZ=True))
    Parents=((Item=(Type=Bone,Name="extra_strap_l_bk"),Weight=0.4995),(Item=(Type=Bone,Name="extra_strap_l_fr"),Weight=0.5005))
    Weight=@n17.OutputParam
   )
   > &RigUnit_ModifyTransforms(
    ItemToModify=((Item=(Type=Bone,Name="dyn_gun_base"),Transform=(Rotation=(X=-0.0905,Y=-0.0019,Z=0.006,W=0.9959),Translation=(X=2.3341,Y=-2.5628,Z=-1.5773),Scale3D=(X=1,Y=1,Z=1))))
    WeightMinimum=0
    Weight=@n17.OutputParam
   )
   n18 = &RigUnit_SphericalPoseReader(
    ActiveRegionScaleFactors=(PositiveWidth=3.5,NegativeWidth=0.5,PositiveHeight=3,NegativeHeight=3)
    ActiveRegionSize=0.1
    Debug=(bDrawDebug=true,DebugScale=25,DebugSegments=20,DebugThickness=0.25)
    DriverAxis=(X=1,Y=0,Z=0)
    DriverItem=(Type=Bone,Name="spine_05")
    FalloffRegionScaleFactors=(PositiveWidth=1,NegativeWidth=0.5,PositiveHeight=1,NegativeHeight=1)
    FalloffSize=0.45
    OptionalParentItem=(Type=Bone,Name="pelvis")
    RotationOffset=(X=90,Y=0,Z=0)
    ExecutePin=@RigVMFunction_Sequence_12.C
    Debug.bDrawDebug=RigVMVariableNode() {variable=Degub_spine}.Value
   )
   > &RigUnit_ModifyTransforms(
    ItemToModify=((Item=(Type=Bone,Name="extra_strap_center_fr"),Transform=(Rotation=(X=0,Y=0,Z=0.1736,W=0.9848),Translation=(X=0,Y=0,Z=0),Scale3D=(X=1,Y=1,Z=1))))
    WeightMinimum=0
    Weight=@n18.OutputParam
   )
   FN_DeformRig_FNC = RigVMFunctionReferenceNode(
    DeformRig_ Curve="(deformRig_l_upperarm_up_alpha_72_#1=False,deformRig_r_upperarm_up_alpha_73_#2=False,deformRig_l_upperarm_fr_alpha_74_#3=False,deformRig_r_upperarm_fr_alpha_75_CB51B8 ...(truncated)"
    DeformRig_Bool="(upperarm_1_#4=True,elbow_3_#5=True,wrist_7_#6=True,groin_8_#7=True,glute_17_#8=True,knee_9_40C408324CA261 ...(truncated)"
    DeformRig_Debug=(upperarm_debug_18_#4=False,wrist_debug_20_#6=False,lat_debug_26_#9=False,groinGlute_debug_27_#7=False)
    DeformRig_Value=RigVMVariableNode() {variable=deform_value}.Value
    ExecuteContext=@RigVMFunction_Sequence.A
   )
  ]}
]
hierarchy [
 root {loc=<0 0 0>, rot=<0 0 0 1>}
 pelvis {parent=root, loc=<0 -0.5301 98.0571>, rot=<0.0127 0.707 -0.0127 -0.707>}
 spine_01 {parent=pelvis, loc=<6.1208 -0.9328 0>, rot=<0 0 0.034 -0.9994>}
 spine_02 {parent=spine_01, loc=<6.5996 -0.2095 0>, rot=<0 0 0.0238 -0.9997>}
 spine_03 {parent=spine_02, loc=<6.4914 -0.1237 0>, rot=<0 0 0.0149 -0.9999>}
 spine_04 {parent=spine_03, loc=<6.9598 0.0794 0>, rot=<0 0 -0.0132 -0.9999>}
 spine_05 {parent=spine_04, loc=<6.9987 0 0>, rot=<0 0 -0.0015 -1>}
 attach_cape {parent=spine_05, loc=<9.322 -8.45 0.0001>, rot=<0.0616 -0.7044 -0.0616 -0.7044>, scale=<1.0001 0.9999 1>}
 clavicle_l {parent=spine_05, loc=<5.4831 -0.1053 -3.2946>, rot=<0.0465 -0.7289 -0.009 -0.6829>}
 upperarm_l {parent=clavicle_l, loc=<9.8658 -1.2136 -0.4105>, rot=<0.0164 -0.3245 -0.0269 -0.9454>}
 lowerarm_l {parent=upperarm_l, loc=<26.465 -0.8345 0.259>, rot=<0.0036 0.0161 -0.2298 -0.9731>}
 hand_l {parent=lowerarm_l, loc=<22.7355 -0.3203 -0.6105>, rot=<-0.6784 -0.0223 0.0735 -0.7307>}
 thumb_01_l {parent=hand_l, loc=<1.3118 -0.3191 -2.459>, rot=<0.6546 -0.4042 0.0664 -0.6354>}
 thumb_02_l {parent=thumb_01_l, loc=<5.9336 -0.0448 0.0401>, rot=<-0.0037 0.0213 0.298 -0.9543>}
 thumb_03_l {parent=thumb_02_l, loc=<2.2223 0.3682 -0.0907>, rot=<0.0118 -0.013 0.0066 -0.9998>}
 index_metacarpal_l {parent=hand_l, loc=<1.6535 0.0183 -2.96>, rot=<0.0158 -0.0124 -0.0173 -0.9996>}
 index_01_l {parent=index_metacarpal_l, loc=<8.3008 0.0357 -0.0524>, rot=<-0.0044 0.0485 0.0924 -0.9945>}
 index_02_l {parent=index_01_l, loc=<3.3587 -0.3004 0.0452>, rot=<-0.002 0.0027 0.3225 -0.9466>}
 index_03_l {parent=index_02_l, loc=<2.5578 -0.123 -0.0252>, rot=<-0.002 0.0067 0.2558 -0.9667>}
 middle_metacarpal_l {parent=hand_l, loc=<2.1856 -0.1725 -0.5984>, rot=<0.006 0.0168 -0.0329 -0.9993>}
 middle_01_l {parent=middle_metacarpal_l, loc=<7.3275 -0.2447 0.0129>, rot=<-0.0208 0.0042 0.0858 -0.9961>}
 middle_02_l {parent=middle_01_l, loc=<4.0123 0.1401 0.1175>, rot=<-0.0218 0.0258 0.3512 -0.9357>}
 middle_03_l {parent=middle_02_l, loc=<2.6077 0.0534 0.1373>, rot=<0.0082 0.0188 0.1708 -0.9851>}
 ring_metacarpal_l {parent=hand_l, loc=<2.2074 -0.3625 1.3638>, rot=<-0.0284 0.0722 -0.0458 -0.9959>}
 ring_01_l {parent=ring_metacarpal_l, loc=<7.0295 -0.6197 -0.1446>, rot=<-0.0489 -0.0252 0.1892 -0.9804>}
 ring_02_l {parent=ring_01_l, loc=<3.6174 0.4647 -0.1027>, rot=<0 0 0.2806 -0.9598>}
 ring_03_l {parent=ring_02_l, loc=<2.4152 0.114 0.0199>, rot=<0 0 0.1189 -0.9929>}
 pinky_metacarpal_l {parent=hand_l, loc=<1.7616 -0.9753 3.3814>, rot=<-0.0397 0.0814 -0.069 -0.9935>}
 pinky_01_l {parent=pinky_metacarpal_l, loc=<7.1709 -0.4831 -0.1094>, rot=<-0.0654 -0.0035 0.2242 -0.9723>}
 pinky_02_l {parent=pinky_01_l, loc=<2.8274 0.0893 -0.1043>, rot=<0.0037 0.0011 0.2854 -0.9584>}
 pinky_03_l {parent=pinky_02_l, loc=<2.2046 0.0307 -0.0078>, rot=<-0.0002 0.0098 0.1445 -0.9895>}
 deform_wrist_dn_base_l {parent=hand_l, loc=<0 -1.5326 0>, rot=<0 0 0 1>}
 deform_wrist_dn_l {parent=deform_wrist_dn_base_l, loc=<0 0 0>, rot=<0 0 0 1>}
 deform_wrist_up_base_l {parent=hand_l, loc=<0 2.3009 0>, rot=<0 0 0 1>}
 deform_wrist_up_l {parent=deform_wrist_up_base_l, loc=<0 0 0>, rot=<0 0 0 1>}
 lowerarm_twist_02_l {parent=lowerarm_l, loc=<10.6707 0 0>, rot=<0 0 0 1>}
 lowerarm_twist_01_l {parent=lowerarm_l, loc=<18.316 0 0>, rot=<0 0 0 1>}
 deform_elbow_in_base_l {parent=lowerarm_l, loc=<0.1834 3.6062 -0.2572>, rot=<0 0 0 1>}
 deform_elbow_in_l {parent=deform_elbow_in_base_l, loc=<0 0 0>, rot=<0 0 0 1>}
 deform_elbow_out_base_l {parent=lowerarm_l, loc=<0 -4 0>, rot=<0 0 0 1>}
 deform_elbow_out_l {parent=deform_elbow_out_base_l, loc=<0 0 0>, rot=<0 0 0 1>}
 upperarm_twist_01_l {parent=upperarm_l, loc=<7.2135 -0.1134 0>, rot=<0 0 0 1>}
 deform_upperarm_bk_base_l {parent=upperarm_twist_01_l, loc=<-6.4053 -4.3746 0.314>, rot=<0 0 0 1>}
 deform_upperarm_bk_l {parent=deform_upperarm_bk_base_l, loc=<0 0 0>, rot=<0 0 0 1>}
 deform_upperarm_fr_base_l {parent=upperarm_twist_01_l, loc=<-5.6312 5.092 0.3354>, rot=<0 0 0 1>}
 deform_upperarm_fr_l {parent=deform_upperarm_fr_base_l, loc=<0 0 0>, rot=<0 0 0 1>}
 deform_upperarm_up_base_l {parent=upperarm_twist_01_l, loc=<-8.891 0.4825 3.9517>, rot=<0 0 0 1>}
 deform_upperarm_up_l {parent=deform_upperarm_up_base_l, loc=<0 0 0>, rot=<0 0 0 1>}
 upperarm_twist_02_l {parent=upperarm_l, loc=<19.7998 -0.1134 0>, rot=<0 0 0 1>}
 dyn_rose_l_side_base {parent=upperarm_l, loc=<1.1263 0.1718 4.349>, rot=<0.5834 0.4413 0.4503 -0.512>}
 dyn_rose_l_side {parent=dyn_rose_l_side_base, loc=<0.1541 0 0>, rot=<0 0 0 1>}
 dyn_rose_02_l_fr_base {parent=upperarm_l, loc=<0.4256 5.6902 1.0205>, rot=<-0.5932 -0.6733 -0.0147 0.4411>}
 dyn_rose_02_l_fr {parent=dyn_rose_02_l_fr_base, loc=<0.1541 0 0>, rot=<0 0 0 1>}
 dyn_rose_02_l_bk_base {parent=upperarm_l, loc=<0.09 -5.2618 1.0831>, rot=<0.3634 -0.0538 0.7019 -0.6102>}
 dyn_rose_02_l_bk {parent=dyn_rose_02_l_bk_base, loc=<0.1541 0 0>, rot=<0 0 0 1>}
 dyn_rose_01_l_bk_base {parent=clavicle_l, loc=<1.4603 -8.0748 -0.1305>, rot=<0.5201 -0.3879 0.5683 -0.5059>}
 dyn_rose_01_l_bk {parent=dyn_rose_01_l_bk_base, loc=<0.1541 0 0>, rot=<0 0 0 1>}
 dyn_thorn_strap_l_base {parent=clavicle_l, loc=<3.0219 -0.7584 -0.2619>, rot=<0.6568 0.2589 0.3011 -0.641>}
 dyn_thorn_strap_l {parent=dyn_thorn_strap_l_base, loc=<0.1836 0 0>, rot=<0 0 0 1>}
 dyn_rose_01_l_fr_base {parent=clavicle_l, loc=<2.5004 7.2972 -1.247>, rot=<0.4212 0.6162 -0.1984 -0.6352>}
 dyn_rose_01_l_fr {parent=dyn_rose_01_l_fr_base, loc=<0.1541 0 0>, rot=<0 0 0 1>}
 attach_backpack {parent=spine_05, loc=<0 -8.35 0>, rot=<-0.0093 -0.707 0.0093 -0.707>, scale=<1.0001 0.9999 1>}
 neck_01 {parent=spine_05, loc=<11.3138 0.1529 0>, rot=<0 0 -0.2003 -0.9797>}
 neck_02 {parent=neck_01, loc=<4.5627 0 0>, rot=<0 0 0.0504 -0.9987>}
 head {parent=neck_02, loc=<5.3914 0 0>, rot=<0 0 0.115 -0.9934>}
 dyn_topknot_base {parent=head, loc=<12.9875 -0.9626 0>, rot=<0 0 -0.1751 -0.9846>}
 dyn_topknot {parent=dyn_topknot_base, loc=<0.2963 0 0>, rot=<0 0 0 1>}
 dyn_braid_01_l_base {parent=head, loc=<4.8248 5.2565 -6.1473>, rot=<-0.0673 0.7034 0.7057 -0.0523>}
 dyn_braid_01_l {parent=dyn_braid_01_l_base, loc=<0.1383 0 0>, rot=<0 0 0 1>}
 dyn_braid_02_l {parent=dyn_braid_01_l, loc=<3.641 0 0>, rot=<-0.0003 0.0603 -0.005 -0.9982>}
 dyn_braid_03_l {parent=dyn_braid_02_l, loc=<3.6413 0.0599 -0.0067>, rot=<-0.0002 0.0197 -0.0102 -0.9998>}
 dyn_braid_04_l {parent=dyn_braid_03_l, loc=<4.0505 0 0>, rot=<0 0 0 1>}
 dyn_braid_01_r_base {parent=head, loc=<4.5784 5.869 6.4634>, rot=<0.0579 -0.7111 0.6999 -0.035>}
 dyn_braid_01_r {parent=dyn_braid_01_r_base, loc=<0.1895 0 0>, rot=<0 0 0 1>}
 dyn_braid_02_r {parent=dyn_braid_01_r, loc=<3.445 0 0>, rot=<-0.0099 -0.0241 -0.0295 -0.9992>}
 dyn_braid_03_r {parent=dyn_braid_02_r, loc=<3.7258 0 0>, rot=<0.0097 -0.0172 0.0018 -0.9998>}
 dyn_braid_04_r {parent=dyn_braid_03_r, loc=<4.0505 0 0>, rot=<0 0 0 1>}
 clavicle_r {parent=spine_05, loc=<5.4831 -0.1053 3.2946>, rot=<0.7289 0.0465 -0.6829 0.009>}
 upperarm_r {parent=clavicle_r, loc=<-9.8658 1.2136 0.4105>, rot=<0.0164 -0.3245 -0.0269 -0.9454>}
 lowerarm_r {parent=upperarm_r, loc=<-26.465 0.8345 -0.259>, rot=<0.0036 0.0161 -0.2298 -0.9731>}
 hand_r {parent=lowerarm_r, loc=<-22.7355 0.3203 0.6106>, rot=<-0.6784 -0.0223 0.0735 -0.7307>}
 thumb_01_r {parent=hand_r, loc=<-1.3118 0.3191 2.459>, rot=<0.6546 -0.4042 0.0664 -0.6354>}
 thumb_02_r {parent=thumb_01_r, loc=<-5.9336 0.0448 -0.0401>, rot=<-0.0037 0.0213 0.298 -0.9543>}
 thumb_03_r {parent=thumb_02_r, loc=<-2.2223 -0.3682 0.0907>, rot=<0.0118 -0.013 0.0066 -0.9998>}
 index_metacarpal_r {parent=hand_r, loc=<-1.6535 -0.0183 2.96>, rot=<0.0158 -0.0124 -0.0173 -0.9996>}
 index_01_r {parent=index_metacarpal_r, loc=<-8.3008 -0.0357 0.0524>, rot=<-0.0044 0.0485 0.0924 -0.9945>}
 index_02_r {parent=index_01_r, loc=<-3.3587 0.3004 -0.0452>, rot=<-0.002 0.0027 0.3225 -0.9466>}
 index_03_r {parent=index_02_r, loc=<-2.5578 0.123 0.0252>, rot=<-0.002 0.0067 0.2558 -0.9667>}
 middle_metacarpal_r {parent=hand_r, loc=<-2.1856 0.1725 0.5984>, rot=<0.006 0.0168 -0.0329 -0.9993>}
 middle_01_r {parent=middle_metacarpal_r, loc=<-7.3275 0.2447 -0.0129>, rot=<-0.0208 0.0042 0.0858 -0.9961>}
 middle_02_r {parent=middle_01_r, loc=<-4.0123 -0.1401 -0.1175>, rot=<-0.0218 0.0258 0.3512 -0.9357>}
 middle_03_r {parent=middle_02_r, loc=<-2.6077 -0.0534 -0.1373>, rot=<0.0082 0.0188 0.1708 -0.9851>}
 ring_metacarpal_r {parent=hand_r, loc=<-2.2074 0.3625 -1.3638>, rot=<-0.0284 0.0722 -0.0458 -0.9959>}
 ring_01_r {parent=ring_metacarpal_r, loc=<-7.0295 0.6197 0.1446>, rot=<-0.0489 -0.0252 0.1892 -0.9804>}
 ring_02_r {parent=ring_01_r, loc=<-3.6174 -0.4647 0.1027>, rot=<0 0 0.2806 -0.9598>}
 ring_03_r {parent=ring_02_r, loc=<-2.4152 -0.114 -0.0199>, rot=<0 0 0.1189 -0.9929>}
 pinky_metacarpal_r {parent=hand_r, loc=<-1.7616 0.9753 -3.3814>, rot=<-0.0397 0.0814 -0.069 -0.9935>}
 pinky_01_r {parent=pinky_metacarpal_r, loc=<-7.1709 0.4831 0.1094>, rot=<-0.0654 -0.0035 0.2242 -0.9723>}
 pinky_02_r {parent=pinky_01_r, loc=<-2.8274 -0.0893 0.1043>, rot=<0.0037 0.0011 0.2854 -0.9584>}
 pinky_03_r {parent=pinky_02_r, loc=<-2.2046 -0.0307 0.0078>, rot=<-0.0002 0.0098 0.1445 -0.9895>}
 deform_wrist_dn_base_r {parent=hand_r, loc=<-0.0001 1.5328 0>, rot=<0 0 0 1>}
 deform_wrist_dn_r {parent=deform_wrist_dn_base_r, loc=<0 0 0>, rot=<0 0 0 1>}
 deform_wrist_up_base_r {parent=hand_r, loc=<-0.0002 -2.3007 -0.0001>, rot=<0 0 0 1>}
 deform_wrist_up_r {parent=deform_wrist_up_base_r, loc=<0 0 0>, rot=<0 0 0 1>}
 lowerarm_twist_02_r {parent=lowerarm_r, loc=<-10.6707 0 0>, rot=<0 0 0 -1>}
 lowerarm_twist_01_r {parent=lowerarm_r, loc=<-18.316 0 0>, rot=<0 0 0 -1>}
 deform_elbow_in_base_r {parent=lowerarm_r, loc=<-0.1834 -3.6062 0.2572>, rot=<0 0 0 -1>}
 deform_elbow_in_r {parent=deform_elbow_in_base_r, loc=<0 0 0>, rot=<0 0 0 1>}
 deform_elbow_out_base_r {parent=lowerarm_r, loc=<0 4 -0.0001>, rot=<0 0 0 -1>}
 deform_elbow_out_r {parent=deform_elbow_out_base_r, loc=<0 0 0>, rot=<0 0 0 1>}
 upperarm_twist_01_r {parent=upperarm_r, loc=<-7.2135 0.1134 0>, rot=<0 0 0 -1>}
 deform_upperarm_bk_base_r {parent=upperarm_twist_01_r, loc=<6.4053 4.3746 -0.314>, rot=<0 0 0 1>}
 deform_upperarm_bk_r {parent=deform_upperarm_bk_base_r, loc=<0 0 0>, rot=<0 0 0 1>}
 deform_upperarm_fr_base_r {parent=upperarm_twist_01_r, loc=<5.6312 -5.092 -0.3354>, rot=<0 0 0 1>}
 deform_upperarm_fr_r {parent=deform_upperarm_fr_base_r, loc=<0 0 0>, rot=<0 0 0 1>}
 deform_upperarm_up_base_r {parent=upperarm_twist_01_r, loc=<8.891 -0.4825 -3.9517>, rot=<0 0 0 1>}
 deform_upperarm_up_r {parent=deform_upperarm_up_base_r, loc=<0 0 0>, rot=<0 0 0 1>}
 upperarm_twist_02_r {parent=upperarm_r, loc=<-19.7998 0.1134 0>, rot=<0 0 0 -1>}
 dyn_rose_r_side_base {parent=upperarm_r, loc=<-1.1262 -0.1718 -4.3488>, rot=<0.4601 -0.5748 -0.5215 -0.4312>}
 dyn_rose_r_side {parent=dyn_rose_r_side_base, loc=<0.1541 0 0>, rot=<0 0 0 1>}
 dyn_rose_02_r_fr_base {parent=upperarm_r, loc=<-1.0702 -5.4545 -0.7946>, rot=<0.7 -0.595 -0.3803 -0.1065>}
 dyn_rose_02_r_fr {parent=dyn_rose_02_r_fr_base, loc=<0.1541 0 0>, rot=<0 0 0 1>}
 dyn_rose_02_r_bk_base {parent=upperarm_r, loc=<1.0119 5.9541 -0.6563>, rot=<-0.0313 -0.3342 -0.6494 -0.6824>}
 dyn_rose_02_r_bk {parent=dyn_rose_02_r_bk_base, loc=<0.1541 0 0>, rot=<0 0 0 1>}
 dyn_rose_01_r_bk_base {parent=clavicle_r, loc=<-1.7168 8.0529 1.0924>, rot=<-0.3245 -0.6079 -0.3982 -0.6055>}
 dyn_rose_01_r_bk {parent=dyn_rose_01_r_bk_base, loc=<0.1541 0 0>, rot=<0 0 0 1>}
 dyn_thorn_strap_r_base {parent=clavicle_r, loc=<-3.0925 0.7761 0.3894>, rot=<0.274 -0.6633 -0.6306 -0.2955>}
 dyn_thorn_strap_r {parent=dyn_thorn_strap_r_base, loc=<0.1836 0 0>, rot=<0 0 0 1>}
 dyn_rose_01_r_fr_base {parent=clavicle_r, loc=<-4.1703 -7.1805 1.7329>, rot=<0.681 -0.3265 -0.6123 0.2338>}
 dyn_rose_01_r_fr {parent=dyn_rose_01_r_fr_base, loc=<0.1541 0 0>, rot=<0 0 0 1>}
 dyn_rose_center_fr_base {parent=spine_05, loc=<3.0584 8.1637 0.3931>, rot=<-0.0251 -0.0054 -0.5069 -0.8616>}
 dyn_rose_center_fr {parent=dyn_rose_center_fr_base, loc=<0.1541 0 0>, rot=<0 0 0 1>}
 dyn_rose_center_bk_down_base {parent=spine_05, loc=<0.8917 -9.4861 -0.7942>, rot=<0.6932 -0.7161 0.058 0.0577>}
 dyn_rose_center_bk {parent=dyn_rose_center_bk_down_base, loc=<0.1541 0 0>, rot=<0 0 0 1>}
 dyn_rose_center_fr_down_base {parent=spine_05, loc=<-2.0261 10.8972 -0.3677>, rot=<-0.0122 0.0047 -0.5338 -0.8455>}
 dyn_rose_center_fr_down {parent=dyn_rose_center_fr_down_base, loc=<0.1541 0 0>, rot=<0 0 0 1>}
 dyn_rose_01_r_fr_down_base {parent=spine_05, loc=<-0.4349 9.0545 7.1133>, rot=<-0.1365 0.0022 -0.5923 -0.7941>}
 dyn_rose_01_r_fr_down {parent=dyn_rose_01_r_fr_down_base, loc=<0.1541 0 0>, rot=<0 0 0 1>}
 dyn_rose_01_r_bk_down_base {parent=spine_05, loc=<2.4464 -9.274 7.9682>, rot=<0.6595 -0.7346 0.1164 0.1085>}
 dyn_rose_01_r_bk_down {parent=dyn_rose_01_r_bk_down_base, loc=<0.1541 0 0>, rot=<0 0 0 1>}
 dyn_rose_01_l_fr_down_base {parent=spine_05, loc=<-0.6014 9.4327 -7.1789>, rot=<0.2204 -0.1662 -0.5573 -0.7831>}
 dyn_rose_01_l_fr_down {parent=dyn_rose_01_l_fr_down_base, loc=<0.1541 0 0>, rot=<0 0 0 1>}
 dyn_rose_01_l_bk_down_base {parent=spine_05, loc=<1.1959 -9.3879 -8.3068>, rot=<0.646 -0.7509 -0.0667 -0.1202>}
 dyn_rose_01_l_bk_down {parent=dyn_rose_01_l_bk_down_base, loc=<0.1541 0 0>, rot=<0 0 0 1>}
 deform_pec_aim_base_l {parent=spine_04, loc=<3.2456 9.3309 -1.2237>, rot=<0.0603 -0.4033 0.3318 -0.8507>}
 deform_pec_aim_l {parent=deform_pec_aim_base_l, loc=<0 0 0>, rot=<0 0 0 1>}
 deform_pec_base_l {parent=deform_pec_aim_l, loc=<9.3665 0 0>, rot=<0 0 0 1>}
 deform_pec_l {parent=deform_pec_base_l, loc=<0 0 0>, rot=<0 0 0 1>}
 deform_pec_aim_base_r {parent=spine_04, loc=<3.2452 9.3309 1.2237>, rot=<-0.4033 -0.0603 0.8507 0.3318>}
 deform_pec_aim_r {parent=deform_pec_aim_base_r, loc=<0 0 0>, rot=<0 0 0 1>}
 deform_pec_base_r {parent=deform_pec_aim_r, loc=<-9.3667 -0.0003 -0.0004>, rot=<0 0 0 1>}
 deform_pec_r {parent=deform_pec_base_r, loc=<0 0 0>, rot=<0 0 0 1>}
 thigh_l {parent=pelvis, loc=<-1.0855 -0.2109 -8.7053>, rot=<0.0009 0.0355 0.0155 -0.9993>}
 calf_l {parent=thigh_l, loc=<-42.5284 -0.0019 -0.7297>, rot=<0.0006 -0.0161 -0.0391 -0.9991>}
 foot_l {parent=calf_l, loc=<-46.7401 -2.5014 -0.592>, rot=<0.0427 -0.0238 0.0353 -0.9982>}
 ball_l {parent=foot_l, loc=<-3.5658 10.8066 -0.1544>, rot=<0.0001 -0.0001 -0.7063 -0.7079>}
 calf_twist_02_l {parent=calf_l, loc=<-19.1546 -1.7943 -0.2638>, rot=<0.0237 0.0095 -0.0035 -0.9997>}
 calf_twist_01_l {parent=calf_l, loc=<-35.8483 -1.9197 -0.5794>, rot=<0.0237 0.0095 -0.0035 -0.9997>}
 deform_knee_in_base_l {parent=calf_l, loc=<0.8031 -5.0231 0.2456>, rot=<0 0 0 -1>}
 deform_knee_in_l {parent=deform_knee_in_base_l, loc=<0 0 0>, rot=<0 0 0 1>}
 deform_knee_out_base_l {parent=calf_l, loc=<0 4 0>, rot=<0 0 0 -1>}
 deform_knee_out_l {parent=deform_knee_out_base_l, loc=<0 0 0>, rot=<0 0 0 1>}
 thigh_twist_01_l {parent=thigh_l, loc=<-27.952 0 0>, rot=<0.0013 0 0 -1>}
 deform_groin_base_l {parent=thigh_l, loc=<0.4801 5.6776 -0.2486>, rot=<0 0 0 -1>}
 deform_groin_l {parent=deform_groin_base_l, loc=<0 0 0>, rot=<0 0 0 1>}
 thigh_r {parent=pelvis, loc=<-1.0855 -0.2109 8.7053>, rot=<0.0355 -0.0009 0.9993 0.0155>}
 calf_r {parent=thigh_r, loc=<42.5284 0.0019 0.7297>, rot=<0.0006 -0.0161 -0.0391 -0.9991>}
 foot_r {parent=calf_r, loc=<46.7401 2.5014 0.592>, rot=<0.0427 -0.0238 0.0353 -0.9982>}
 ball_r {parent=foot_r, loc=<3.5658 -10.8066 0.1544>, rot=<0.0001 -0.0001 -0.7063 -0.7079>}
 calf_twist_02_r {parent=calf_r, loc=<19.1546 1.7943 0.2638>, rot=<0.0237 0.0095 -0.0035 -0.9997>}
 calf_twist_01_r {parent=calf_r, loc=<35.8483 1.9197 0.5794>, rot=<0.0237 0.0095 -0.0035 -0.9997>}
 deform_knee_in_base_r {parent=calf_r, loc=<-0.8031 5.0231 -0.2457>, rot=<0 0 0 -1>}
 deform_knee_in_r {parent=deform_knee_in_base_r, loc=<0 0 0>, rot=<0 0 0 1>}
 deform_knee_out_base_r {parent=calf_r, loc=<0 -4 0>, rot=<0 0 0 -1>}
 deform_knee_out_r {parent=deform_knee_out_base_r, loc=<0 0 0>, rot=<0 0 0 1>}
 thigh_twist_01_r {parent=thigh_r, loc=<27.952 0 0>, rot=<0.0013 0 0 -1>}
 deform_groin_base_r {parent=thigh_r, loc=<-0.4801 -5.6776 0.2486>, rot=<0 0 0 -1>}
 deform_groin_r {parent=deform_groin_base_r, loc=<0 0 0>, rot=<0 0 0 1>}
 extra_strap_center_fr {parent=pelvis, loc=<3.8638 9.5766 -0.7183>, rot=<-0.0242 -0.035 -0.6676 -0.7433>}
 extra_strap_l_fr {parent=pelvis, loc=<1.4698 5.6858 -13.2377>, rot=<0.3235 -0.2281 -0.5612 -0.7269>}
 extra_strap_l_bk {parent=pelvis, loc=<3.2925 -4.8058 -13.9577>, rot=<0.7477 -0.4097 -0.3523 -0.386>}
 extra_strap_r_fr {parent=pelvis, loc=<4.7857 4.7076 10.6662>, rot=<-0.3889 0.114 -0.5742 -0.7114>}
 extra_strap_r_bk {parent=pelvis, loc=<6.6241 -2.4882 12.0952>, rot=<0.7221 -0.3865 0.306 0.4854>}
 deform_glute_base_l {parent=pelvis, loc=<-7.3466 -9.0213 -9.1245>, rot=<0.0172 0.031 -0.4477 -0.8935>}
 deform_glute_l {parent=deform_glute_base_l, loc=<0 0 0>, rot=<0 0 0 1>}
 deform_glute_base_r {parent=pelvis, loc=<-7.3464 -9.0213 9.1245>, rot=<0.031 -0.0172 0.8935 -0.4477>}
 deform_glute_r {parent=deform_glute_base_r, loc=<0 0 0>, rot=<0 0 0 1>}
 dyn_gun_base {parent=pelvis, loc=<2.5355 0.6445 -15.4081>, rot=<0.5687 -0.4066 -0.442 -0.562>}
 dyn_gun {parent=dyn_gun_base, loc=<0.2587 0 0>, rot=<0 0 0 1>}
]
variables [
 Degub_arm_l {type=bool, default=False}
 Degub_arm_r {type=bool, default=False}
 Degub_spine {type=bool, default=False}
 Degub_thigh_l {type=bool, default=False}
 Degub_thigh_r {type=bool, default=False}
 deform_value {
  type=FUserDefinedStruct_#10
  default="(upperarm_up_18_#11=(Rotation=(X=0,Y=0,Z=0,W=1),Translation=(X=0,Y=0,Z=8),Scale3D=(X=0.7,Y=1,Z=1)),upperarm_fwd_fr_19_#12=(Rotation=(X=0,Y=0,Z=0,W=1),Translation=(X=0,Y=10,Z=0),Scale3D=(X=0.5,Y=1,Z= ...(truncated)"
 }
]
