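# Physics asset for the female medium ThornedEden body cosmetic, covering its dyn_rose bones around the upper torso and shoulders, apparently for secondary motion.
# Each rose capsule is constrained to a tiny base capsule with narrow swing and near-zero twist.
# The bases attach with free rotation to spine_05, clavicle or upper arm capsules.
# Only these upper-body bones and neck_01 have body capsules, so it is not a full-body ragdoll.
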PhysicsAsset /BRCosmetics/Characters/Player/Female/Medium/Bodies/F_MED_ThornedEden/Meshes/F_MED_ThornedEden_Physics
bodies {
 clavicle_l [{shapeType=Capsule, length=5, radius=5, shapeName=clavicle_l_capsule, rotation=<rot 90 -87.106 -87.1062>}]
 clavicle_r [{shapeType=Capsule, length=5, radius=5, shapeName=clavicle_r_capsule, rotation=<rot 90 -87.106 -87.1062>}]
 dyn_rose_01_l_bk [{shapeType=Capsule, length=1, radius=4, shapeName=dyn_rose_01_l_bk_capsule, center=<5 0 0>, rotation=<rot 90 0 0>}]
 dyn_rose_01_l_bk_base [{shapeType=Capsule, length=0.505, radius=0.505, shapeName=dyn_rose_01_l_bk_base_capsule}]
 dyn_rose_01_l_bk_down [{shapeType=Capsule, length=1, radius=4, shapeName=dyn_rose_01_l_bk_down_capsule, center=<5 0 0>, rotation=<rot 90 0 0>}]
 dyn_rose_01_l_bk_down_base [{shapeType=Capsule, length=0.505, radius=0.505, shapeName=dyn_rose_01_l_bk_down_base_capsule, rotation=<rot 0 -90 0>}]
 dyn_rose_01_l_fr [{shapeType=Capsule, length=1, radius=4, shapeName=dyn_rose_01_l_fr_capsule, center=<5 0 0>, rotation=<rot 90 0 0>}]
 dyn_rose_01_l_fr_base [{shapeType=Capsule, length=0.505, radius=0.505, shapeName=dyn_rose_01_l_fr_base_capsule}]
 dyn_rose_01_l_fr_down [{shapeType=Capsule, length=1, radius=4, shapeName=dyn_rose_01_l_fr_down_capsule, center=<5 0 0>, rotation=<rot 90 0 0>}]
 dyn_rose_01_l_fr_down_base [{shapeType=Capsule, length=0.505, radius=0.505, shapeName=dyn_rose_01_l_fr_down_base_capsule, rotation=<rot 0 -90 0>}]
 dyn_rose_01_r_bk [{shapeType=Capsule, length=1, radius=4, shapeName=dyn_rose_01_r_bk_capsule, center=<5 0 0>, rotation=<rot 90 0 0>}]
 dyn_rose_01_r_bk_base [{shapeType=Capsule, length=0.505, radius=0.505, shapeName=dyn_rose_01_r_bk_base_capsule}]
 dyn_rose_01_r_bk_down [{shapeType=Capsule, length=1, radius=4, shapeName=dyn_rose_01_r_bk_down_capsule, center=<5 0 0>, rotation=<rot 90 0 0>}]
 dyn_rose_01_r_bk_down_base [{shapeType=Capsule, length=0.505, radius=0.505, shapeName=dyn_rose_01_r_bk_down_base_capsule, rotation=<rot 0 -90 0>}]
 dyn_rose_01_r_fr [{shapeType=Capsule, length=1, radius=4, shapeName=dyn_rose_01_r_fr_capsule, center=<5 0 0>, rotation=<rot 90 0 0>}]
 dyn_rose_01_r_fr_base [{shapeType=Capsule, length=0.505, radius=0.505, shapeName=dyn_rose_01_r_fr_base_capsule}]
 dyn_rose_01_r_fr_down [{shapeType=Capsule, length=1, radius=4, shapeName=dyn_rose_01_r_fr_down_capsule, center=<5 0 0>, rotation=<rot 90 0 0>}]
 dyn_rose_01_r_fr_down_base [{shapeType=Capsule, length=0.505, radius=0.505, shapeName=dyn_rose_01_r_fr_down_base_capsule, rotation=<rot 0 -90 0>}]
 dyn_rose_02_l_bk [{shapeType=Capsule, length=1, radius=4, shapeName=dyn_rose_02_l_bk_capsule, center=<5 0 0>, rotation=<rot 90 0 0>}]
 dyn_rose_02_l_bk_base [{shapeType=Capsule, length=0.505, radius=0.505, shapeName=dyn_rose_02_l_bk_base_capsule}]
 dyn_rose_02_l_fr [{shapeType=Capsule, length=1, radius=4, shapeName=dyn_rose_02_l_fr_capsule, center=<5 0 0>, rotation=<rot 90 0 0>}]
 dyn_rose_02_l_fr_base [{shapeType=Capsule, length=0.505, radius=0.505, shapeName=dyn_rose_02_l_fr_base_capsule}]
 dyn_rose_02_r_bk [{shapeType=Capsule, length=1, radius=4, shapeName=dyn_rose_02_r_bk_capsule, center=<5 0 0>, rotation=<rot 90 0 0>}]
 dyn_rose_02_r_bk_base [{shapeType=Capsule, length=0.505, radius=0.505, shapeName=dyn_rose_02_r_bk_base_capsule}]
 dyn_rose_02_r_fr [{shapeType=Capsule, length=1, radius=4, shapeName=dyn_rose_02_r_fr_capsule, center=<5 0 0>, rotation=<rot 90 0 0>}]
 dyn_rose_02_r_fr_base [{shapeType=Capsule, length=0.505, radius=0.505, shapeName=dyn_rose_02_r_fr_base_capsule}]
 dyn_rose_center_bk [{shapeType=Capsule, length=1, radius=4, shapeName=dyn_rose_center_bk_capsule, center=<5 0 0>, rotation=<rot 90 0 0>}]
 dyn_rose_center_bk_down_base [{shapeType=Capsule, length=0.505, radius=0.505, shapeName=dyn_rose_center_bk_down_base_capsule, rotation=<rot 0 -90 0>}]
 dyn_rose_center_fr [{shapeType=Capsule, length=1, radius=4, shapeName=dyn_rose_center_fr_capsule, center=<5 0 0>, rotation=<rot 90 0 0>}]
 dyn_rose_center_fr_base [{shapeType=Capsule, length=0.505, radius=0.505, shapeName=dyn_rose_center_fr_base_capsule}]
 dyn_rose_center_fr_down [{shapeType=Capsule, length=1, radius=4, shapeName=dyn_rose_center_fr_down_capsule, center=<5 0 0>, rotation=<rot 90 0 0>}]
 dyn_rose_center_fr_down_base [{shapeType=Capsule, length=0.505, radius=0.505, shapeName=dyn_rose_center_fr_down_base_capsule, rotation=<rot 0 -90 0>}]
 dyn_rose_l_side [{shapeType=Capsule, length=1, radius=4, shapeName=dyn_rose_l_side_capsule, center=<5 0 0>, rotation=<rot 90 0 0>}]
 dyn_rose_l_side_base [{shapeType=Capsule, length=0.505, radius=0.505, shapeName=dyn_rose_l_side_base_capsule}]
 dyn_rose_r_side [{shapeType=Capsule, length=1, radius=4, shapeName=dyn_rose_r_side_capsule, center=<5 0 0>, rotation=<rot 90 0 0>}]
 dyn_rose_r_side_base [{shapeType=Capsule, length=0.505, radius=0.505, shapeName=dyn_rose_r_side_base_capsule}]
 neck_01 [{shapeType=Capsule, length=5, radius=5, shapeName=neck_01_capsule, rotation=<rot 90 -90 90>}]
 spine_05 [{shapeType=Capsule, length=5, radius=5, shapeName=spine_05_capsule, rotation=<rot 90 0 0.0002>}]
 upperarm_l [{shapeType=Capsule, length=20, radius=5, shapeName=upperarm_l_capsule, center=<10 0 0>, rotation=<rot 90 -179.9602 -179.9606>}]
 upperarm_r [{shapeType=Capsule, length=5, radius=5, shapeName=upperarm_r_capsule, rotation=<rot 90 -90.9147 -90.915>}]
}
constraints [
 {bone1Name=dyn_rose_01_l_bk, bone2Name=dyn_rose_01_l_bk_base, swing1LimitDegrees=15, swing2LimitDegrees=15, twistLimitDegrees=1}
 {bone1Name=dyn_rose_02_l_fr, bone2Name=dyn_rose_02_l_fr_base, swing1LimitDegrees=15, swing2LimitDegrees=15, twistLimitDegrees=1}
 {bone1Name=dyn_rose_l_side, bone2Name=dyn_rose_l_side_base, swing1LimitDegrees=15, swing2LimitDegrees=15, twistLimitDegrees=1}
 {bone1Name=dyn_rose_02_r_bk, bone2Name=dyn_rose_02_r_bk_base, swing1LimitDegrees=15, swing2LimitDegrees=15, twistLimitDegrees=1}
 {bone1Name=dyn_rose_02_l_bk, bone2Name=dyn_rose_02_l_bk_base, swing1LimitDegrees=15, swing2LimitDegrees=15, twistLimitDegrees=1}
 {bone1Name=dyn_rose_01_r_fr, bone2Name=dyn_rose_01_r_fr_base, swing1LimitDegrees=15, swing2LimitDegrees=15, twistLimitDegrees=1}
 {bone1Name=dyn_rose_center_fr, bone2Name=dyn_rose_center_fr_base, swing1LimitDegrees=15, swing2LimitDegrees=15, twistLimitDegrees=1}
 {bone1Name=dyn_rose_01_l_fr, bone2Name=dyn_rose_01_l_fr_base, swing1LimitDegrees=15, swing2LimitDegrees=15, twistLimitDegrees=1}
 {bone1Name=dyn_rose_r_side, bone2Name=dyn_rose_r_side_base, swing1LimitDegrees=15, swing2LimitDegrees=15, twistLimitDegrees=1}
 {bone1Name=dyn_rose_01_r_bk, bone2Name=dyn_rose_01_r_bk_base, swing1LimitDegrees=15, swing2LimitDegrees=15, twistLimitDegrees=1}
 {bone1Name=dyn_rose_02_r_fr, bone2Name=dyn_rose_02_r_fr_base, swing1LimitDegrees=15, swing2LimitDegrees=15, twistLimitDegrees=1}
 {
  bone1Name=dyn_rose_l_side_base
  bone2Name=upperarm_l
  swing1LimitDegrees=45
  swing1Motion=Free
  swing2LimitDegrees=45
  swing2Motion=Free
  twistLimitDegrees=45
  twistMotion=Free
 }
 {
  bone1Name=dyn_rose_01_l_bk_base
  bone2Name=clavicle_l
  swing1LimitDegrees=45
  swing1Motion=Free
  swing2LimitDegrees=45
  swing2Motion=Free
  twistLimitDegrees=45
  twistMotion=Free
 }
 {
  bone1Name=dyn_rose_02_l_fr_base
  bone2Name=clavicle_l
  swing1LimitDegrees=45
  swing1Motion=Free
  swing2LimitDegrees=45
  swing2Motion=Free
  twistLimitDegrees=45
  twistMotion=Free
 }
 {
  bone1Name=dyn_rose_r_side_base
  bone2Name=upperarm_r
  swing1LimitDegrees=45
  swing1Motion=Free
  swing2LimitDegrees=45
  swing2Motion=Free
  twistLimitDegrees=45
  twistMotion=Free
 }
 {
  bone1Name=dyn_rose_01_r_bk_base
  bone2Name=clavicle_r
  swing1LimitDegrees=45
  swing1Motion=Free
  swing2LimitDegrees=45
  swing2Motion=Free
  twistLimitDegrees=45
  twistMotion=Free
 }
 {
  bone1Name=dyn_rose_02_r_fr_base
  bone2Name=clavicle_r
  swing1LimitDegrees=45
  swing1Motion=Free
  swing2LimitDegrees=45
  swing2Motion=Free
  twistLimitDegrees=45
  twistMotion=Free
 }
 {
  bone1Name=dyn_rose_center_fr_base
  bone2Name=spine_05
  swing1LimitDegrees=45
  swing1Motion=Free
  swing2LimitDegrees=45
  swing2Motion=Free
  twistLimitDegrees=45
  twistMotion=Free
 }
 {
  bone1Name=dyn_rose_01_r_fr_base
  bone2Name=spine_05
  swing1LimitDegrees=45
  swing1Motion=Free
  swing2LimitDegrees=45
  swing2Motion=Free
  twistLimitDegrees=45
  twistMotion=Free
 }
 {
  bone1Name=dyn_rose_02_r_bk_base
  bone2Name=spine_05
  swing1LimitDegrees=45
  swing1Motion=Free
  swing2LimitDegrees=45
  swing2Motion=Free
  twistLimitDegrees=45
  twistMotion=Free
 }
 {
  bone1Name=dyn_rose_01_l_fr_base
  bone2Name=spine_05
  swing1LimitDegrees=45
  swing1Motion=Free
  swing2LimitDegrees=45
  swing2Motion=Free
  twistLimitDegrees=45
  twistMotion=Free
 }
 {
  bone1Name=dyn_rose_02_l_bk_base
  bone2Name=spine_05
  swing1LimitDegrees=45
  swing1Motion=Free
  swing2LimitDegrees=45
  swing2Motion=Free
  twistLimitDegrees=45
  twistMotion=Free
 }
 {
  bone1Name=clavicle_r
  bone2Name=spine_05
  swing1LimitDegrees=45
  swing1Motion=Free
  swing2LimitDegrees=45
  swing2Motion=Free
  twistLimitDegrees=45
  twistMotion=Free
 }
 {
  bone1Name=upperarm_r
  bone2Name=clavicle_r
  swing1LimitDegrees=45
  swing1Motion=Free
  swing2LimitDegrees=45
  swing2Motion=Free
  twistLimitDegrees=45
  twistMotion=Free
 }
 {
  bone1Name=clavicle_l
  bone2Name=spine_05
  swing1LimitDegrees=45
  swing1Motion=Free
  swing2LimitDegrees=45
  swing2Motion=Free
  twistLimitDegrees=45
  twistMotion=Free
 }
 {
  bone1Name=upperarm_l
  bone2Name=clavicle_l
  swing1LimitDegrees=45
  swing1Motion=Free
  swing2LimitDegrees=45
  swing2Motion=Free
  twistLimitDegrees=45
  twistMotion=Free
 }
 {
  bone1Name=neck_01
  bone2Name=spine_05
  swing1LimitDegrees=45
  swing1Motion=Free
  swing2LimitDegrees=45
  swing2Motion=Free
  twistLimitDegrees=45
  twistMotion=Free
 }
 {bone1Name=dyn_rose_center_bk, bone2Name=dyn_rose_center_bk_down_base, swing1LimitDegrees=15, swing2LimitDegrees=15, twistLimitDegrees=1}
 {
  bone1Name=dyn_rose_center_bk_down_base
  bone2Name=spine_05
  swing1LimitDegrees=45
  swing1Motion=Free
  swing2LimitDegrees=45
  swing2Motion=Free
  twistLimitDegrees=45
  twistMotion=Free
 }
 {bone1Name=dyn_rose_01_r_fr_down, bone2Name=dyn_rose_01_r_fr_down_base, swing1LimitDegrees=15, swing2LimitDegrees=15, twistLimitDegrees=1}
 {
  bone1Name=dyn_rose_01_r_fr_down_base
  bone2Name=spine_05
  swing1LimitDegrees=45
  swing1Motion=Free
  swing2LimitDegrees=45
  swing2Motion=Free
  twistLimitDegrees=45
  twistMotion=Free
 }
 {
  bone1Name=dyn_rose_01_r_bk_down_base
  bone2Name=spine_05
  swing1LimitDegrees=45
  swing1Motion=Free
  swing2LimitDegrees=45
  swing2Motion=Free
  twistLimitDegrees=45
  twistMotion=Free
 }
 {bone1Name=dyn_rose_01_r_bk_down, bone2Name=dyn_rose_01_r_bk_down_base, swing1LimitDegrees=15, swing2LimitDegrees=15, twistLimitDegrees=1}
 {bone1Name=dyn_rose_01_l_fr_down, bone2Name=dyn_rose_01_l_fr_down_base, swing1LimitDegrees=15, swing2LimitDegrees=15, twistLimitDegrees=1}
 {
  bone1Name=dyn_rose_01_l_fr_down_base
  bone2Name=spine_05
  swing1LimitDegrees=45
  swing1Motion=Free
  swing2LimitDegrees=45
  swing2Motion=Free
  twistLimitDegrees=45
  twistMotion=Free
 }
 {bone1Name=dyn_rose_01_l_bk_down, bone2Name=dyn_rose_01_l_bk_down_base, swing1LimitDegrees=15, swing2LimitDegrees=15, twistLimitDegrees=1}
 {
  bone1Name=dyn_rose_01_l_bk_down_base
  bone2Name=spine_05
  swing1LimitDegrees=45
  swing1Motion=Free
  swing2LimitDegrees=45
  swing2Motion=Free
  twistLimitDegrees=45
  twistMotion=Free
 }
 {
  bone1Name=dyn_rose_center_fr_down_base
  bone2Name=spine_05
  swing1LimitDegrees=45
  swing1Motion=Free
  swing2LimitDegrees=45
  swing2Motion=Free
  twistLimitDegrees=45
  twistMotion=Free
 }
 {bone1Name=dyn_rose_center_fr_down, bone2Name=dyn_rose_center_fr_down_base, swing1LimitDegrees=15, swing2LimitDegrees=15, twistLimitDegrees=1}
]
