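# Physics asset for the gun on the Thorned Eden medium female player body, giving the dyn_gun bone a setup for physical secondary motion.
# Two capsules shape dyn_gun, which joins a small dyn_gun_base anchor body through tight, asymmetric swing and twist angles.
# Pelvis, upper spine and left thigh capsules approximate the nearby body.
# The anchor, spine and thigh each constrain to the pelvis with free swing and twist.
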PhysicsAsset /BRCosmetics/Characters/Player/Female/Medium/Bodies/F_MED_ThornedEden/Meshes/F_MED_ThornedEden_Physics_gun
bodies {
 dyn_gun [
  {shapeType=Capsule, length=25, radius=5, shapeName=dyn_gun_capsule, center=<3.6341 3.2047 0.0008>, rotation=<rot -0.0009 12.367 89.9868>}
  {shapeType=Capsule, length=7, radius=5, shapeName=dyn_gun_capsule, center=<6.766 -11.8292 -5.6234>, rotation=<rot 13.4699 4.4205 18.3595>}
 ]
 dyn_gun_base [{shapeType=Capsule, length=1, radius=1, shapeName=dyn_gun_base_capsule}]
 pelvis [{shapeType=Capsule, length=15, radius=10, shapeName=pelvis_capsule, center=<-2.3188 -3.152 0>}]
 spine_03 [{shapeType=Capsule, length=7.5509, radius=8.9377, shapeName=spine_03_capsule, center=<1.8194 0 0>, rotation=<rot 0 -90 0>}]
 thigh_l [{shapeType=Capsule, length=20, radius=10, shapeName=thigh_l_capsule, center=<-5.4117 0.0889 1.1861>, rotation=<rot 0 96.8367 -77.9869>}]
}
constraints [
 {
  bone1Name=dyn_gun_base
  bone2Name=pelvis
  swing1LimitDegrees=45
  swing1Motion=Free
  swing2LimitDegrees=45
  swing2Motion=Free
  twistLimitDegrees=45
  twistMotion=Free
 }
 {bone1Name=dyn_gun, bone2Name=dyn_gun_base, swing1LimitDegrees=15, swing2LimitDegrees=10, twistLimitDegrees=5}
 {
  bone1Name=thigh_l
  bone2Name=pelvis
  swing1LimitDegrees=45
  swing1Motion=Free
  swing2LimitDegrees=45
  swing2Motion=Free
  twistLimitDegrees=45
  twistMotion=Free
 }
 {
  bone1Name=spine_03
  bone2Name=pelvis
  swing1LimitDegrees=45
  swing1Motion=Free
  swing2LimitDegrees=45
  swing2Motion=Free
  twistLimitDegrees=45
  twistMotion=Free
 }
]
