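# Post-process style AnimBlueprint for the ThornedEden female body cosmetic, adding secondary motion to the pose copied from the attached parent mesh.
# A Control Rig feeds two RigidBody simulations, body and gun physics assets, then three additive offsets on dyn_gun_base.
# Each update computes gravity through a deprecated override function, feeding the gun simulation.
# It also sets front-end idle flags from emote state, left-arm reference-pose delta and rigid-body enablement.
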AnimBlueprint /BRCosmetics/Characters/Player/Female/Medium/Bodies/F_MED_ThornedEden/Meshes/F_MED_ThornedEden_Skeleton_AnimBlueprint
target_skeleton=&/F_MED_ThornedEden_Skeleton
graphs [
 AnimGraph {nodes [
   n1 = Control Rig(Source=Copy Pose From Mesh() {bUseAttachedParent=true}) {
    alphaBoolBlend {blendOption=Linear}
    controlRigAssetReference {blueprintRigClass=&/F_MED_ThornedEden_CtrlRig.F_MED_ThornedEden_CtrlRig_C}
   }
   n2 = RigidBody!(ComponentPose=Local To Component(LocalPose=@n1), Alpha, OverrideWorldGravity:Vector="0,0,-980", ExternalForce) -> (Pose) {
    bClampLinearTranslationLimitToRefPose=true
    bFreezeIncomingPoseOnStart=true
    bOverrideWorldGravity=true
    bTransferBoneVelocities=true
    overridePhysicsAsset=&/F_MED_ThornedEden_Physics
    simulationSpace=BaseBoneSpace
    alphaBoolBlend {blendOption=Linear}
    baseBoneRef {boneName=spine_05}
    componentAppliedLinearAccClamp=<3000 3000 3000>
    componentLinearAccScale=<0.4 0.4 0.4>
    componentLinearVelScale=<0.4 0.4 0.4>
    overrideWorldGravity {z=-980}
    simSpaceSettings {worldAlpha=0.5}
   }
   n3 = RigidBody(ComponentPose=@n2, OverrideWorldGravity=Get Gravity() -> (Gravity:Vector)) {
    bClampLinearTranslationLimitToRefPose=true
    bFreezeIncomingPoseOnStart=true
    bOverrideWorldGravity=true
    bTransferBoneVelocities=true
    overridePhysicsAsset=&/F_MED_ThornedEden_Physics_gun
    simulationSpace=BaseBoneSpace
    alphaBoolBlend {blendOption=Linear}
    baseBoneRef {boneName=pelvis}
    componentLinearAccScale=<0.4 0.4 0.4>
    componentLinearVelScale=<0.4 0.4 0.4>
    overrideWorldGravity {z=-980}
    simSpaceSettings {worldAlpha=0.5}
   }
   n4 = "Transform (Modify) Bone - Bone: dyn_gun_base"!(ComponentPose=@n3, bAlphaBoolEnabled:bool=True, Translation="1.255,0.957,0.194", Rotation="-26.245,2.408,-6.705", Scale="1,1,1") -> (Pose) {
    alphaInputType=Bool
    rotationMode=BMM_Additive
    rotationSpace=BCS_BoneSpace
    translationMode=BMM_Additive
    translationSpace=BCS_BoneSpace
    alphaBoolBlend {blendInTime=0.8, blendOption=Linear, blendOutTime=0.8}
    boneToModify {boneName=dyn_gun_base}
    rotation=<rot -26.245 2.4081 -6.7051>
    translation=<1.255 0.957 0.194>
   }
   n5 = "Transform (Modify) Bone - Bone: dyn_gun_base"(ComponentPose=@n4, bAlphaBoolEnabled=True, Translation="1.255,0.957,0.194", Rotation="-16.445,6.932,-7.920", Scale="1,1,1") {
    alphaInputType=Bool
    rotationMode=BMM_Additive
    rotationSpace=BCS_BoneSpace
    translationMode=BMM_Additive
    translationSpace=BCS_BoneSpace
    alphaBoolBlend {blendInTime=0.8, blendOption=Linear, blendOutTime=0.8}
    boneToModify {boneName=dyn_gun_base}
    rotation=<rot -16.445 6.932 -7.9201>
    translation=<1.255 0.957 0.194>
   }
   n6 = "Transform (Modify) Bone - Bone: dyn_gun_base"(ComponentPose=@n5, bAlphaBoolEnabled=True, Translation="0,0,0", Rotation="-0.020,-11.539,0.025", Scale="1,1,1") {
    alphaInputType=Bool
    rotationMode=BMM_Additive
    rotationSpace=BCS_BoneSpace
    translationMode=BMM_Additive
    translationSpace=BCS_BoneSpace
    alphaBoolBlend {blendInTime=0.8, blendOption=Linear, blendOutTime=0.8}
    boneToModify {boneName=dyn_gun_base}
    rotation=<rot -0.02 -11.539 0.0249>
   }
   Output Pose(Result=Component To Local(ComponentPose=@n6))
  ]}
 EventGraph {nodes [
   Event BlueprintUpdateAnimation()
   > n1 = Gravity Override State Function DEPRECATED(
    CharacterPartAnimInstance=Self-Reference()
    GameTimeInSeconds=GetGameTimeInSeconds()
    GravityOverrideParams=Get Gravity Override Params()
    RBAN Params=Get RBAN Params()
   )
   > Set Gravity(Gravity:Vector=@n1) -> (Output_Get:Vector)
   n2 = Get bIsPlayingEmote()
   > n3 = Branch(Condition=AND Boolean(A=NOT Boolean(A=@n2), B=Get bIsInFrontEnd()))
   > n4 = Sequence()
   n5 = GetOwningComponent()
   > n6 = Branch(Condition="Equal (Float)"(A=GetAlphaFromReferenceTransformDelta(SkeletalMeshComponent=@n5, BoneName=upperarm_l, BaseName=lowerarm_l, InputRange=(X=-20,Y=-30), TwistAxis="1,0,0")))
   > Set FrontEnd_Idle_first(FrontEnd_Idle_first:bool=true) -> (Output_Get:bool)
   > Set FrontEnd_Idle_second(FrontEnd_Idle_second:bool=false) -> (Output_Get:bool)
   Set FrontEnd_Idle_first(execute=@n6.else, FrontEnd_Idle_first=false)
   > Set FrontEnd_Idle_second(FrontEnd_Idle_second=true)
   n7 = Branch(execute=@n4.then_1, Condition=Get bRigidBodyEnabled())
   > Set FrontEnd_Idle_RBAN(FrontEnd_Idle_RBAN:bool=false) -> (Output_Get:bool)
   Set FrontEnd_Idle_RBAN(execute=@n7.else, FrontEnd_Idle_RBAN=true)
   Set FrontEnd_Idle_first(execute=@n3.else, FrontEnd_Idle_first=false)
   > Set FrontEnd_Idle_second(FrontEnd_Idle_second=false)
   > Set FrontEnd_Idle_RBAN(FrontEnd_Idle_RBAN=false)
   TryGetPawnOwner()
  ]}
]
variables [
 FrontEnd_Idle_first {type=boolean}
 FrontEnd_Idle_RBAN {type=boolean}
 FrontEnd_Idle_second {type=boolean}
 Gravity {type=Vector}
 Gravity Override Params {type=GravityOverrideParamsStruct}
 In String {type=string, default=/Game/Animation/Game/MainPlayer/Menu/Female/Medium/Commando/Commando_Female_Body_Montage_Lobby}
 RBAN Params {
  type=RBANParamsStruct
  default {
   dBNO {simSpaceSettings {
     maxAngularAcceleration=1000
     maxAngularVelocity=3
     maxLinearAcceleration=300
     maxLinearVelocity=700
     velocityScaleZ=0
     externalLinearDragV=<0 0.5 0>
    }}
   dBNO_Carried {simSpaceSettings {
     maxAngularAcceleration=1000
     maxAngularVelocity=3
     maxLinearAcceleration=300
     maxLinearVelocity=700
     velocityScaleZ=0
     externalLinearDragV=<0 0.5 0>
    }}
   emoteOrMelee {simSpaceSettings {
     maxAngularAcceleration=1000
     maxAngularVelocity=3
     maxLinearAcceleration=300
     maxLinearVelocity=700
     velocityScaleZ=0
     externalLinearDragV=<0 0.5 0>
    }}
   front_End {simSpaceSettings {
     maxAngularAcceleration=1000
     maxAngularVelocity=3
     maxLinearAcceleration=300
     maxLinearVelocity=700
     velocityScaleZ=0
     externalLinearDragV=<0 0.5 0>
    }}
   onGround_Crouch_Moving {simSpaceSettings {
     maxAngularAcceleration=1000
     maxAngularVelocity=3
     maxLinearAcceleration=300
     maxLinearVelocity=700
     velocityScaleZ=0
     externalLinearDragV=<0 0.5 0>
    }}
   onGround_Falling {simSpaceSettings {
     maxAngularAcceleration=1000
     maxAngularVelocity=3
     maxLinearAcceleration=300
     maxLinearVelocity=700
     velocityScaleZ=0
     externalLinearDragV=<0 0.5 0>
    }}
   onGround_Moving {simSpaceSettings {
     maxAngularAcceleration=1000
     maxAngularVelocity=3
     maxLinearAcceleration=300
     maxLinearVelocity=700
     velocityScaleZ=0
     externalLinearDragV=<0 0.5 0>
    }}
   onGround_Standing {simSpaceSettings {
     maxAngularAcceleration=1000
     maxAngularVelocity=3
     maxLinearAcceleration=300
     maxLinearVelocity=700
     velocityScaleZ=0
     externalLinearDragV=<0 0.5 0>
    }}
   rail_Grinding {simSpaceSettings {
     maxAngularAcceleration=1000
     maxAngularVelocity=3
     maxLinearAcceleration=300
     maxLinearVelocity=700
     velocityScaleZ=0
     worldAlpha=0.1
     externalLinearDragV=<0 0.5 0>
    }}
   riding_Moving {simSpaceSettings {
     maxAngularAcceleration=1000
     maxAngularVelocity=3
     maxLinearAcceleration=300
     maxLinearVelocity=700
     velocityScaleZ=0
     externalLinearDragV=<0 0.5 0>
    }}
   riding_Moving_Targeting {simSpaceSettings {
     maxAngularAcceleration=1000
     maxAngularVelocity=3
     maxLinearAcceleration=300
     maxLinearVelocity=700
     velocityScaleZ=0
     externalLinearDragV=<0 0.5 0>
    }}
   riding_Standing {simSpaceSettings {
     maxAngularAcceleration=1000
     maxAngularVelocity=3
     maxLinearAcceleration=300
     maxLinearVelocity=700
     velocityScaleZ=0
     externalLinearDragV=<0 0.5 0>
    }}
   skydive_DivingDown {simSpaceSettings {
     maxAngularAcceleration=1000
     maxAngularVelocity=3
     maxLinearAcceleration=300
     maxLinearVelocity=700
     velocityScaleZ=0
     externalLinearDragV=<0 0.5 0>
    }}
   skydive_FloatingDown {
    componentLinearAccScale=<0 0.5 0>
    componentLinearVelScale=<0 0.5 0>
    simSpaceSettings {
     maxAngularAcceleration=1000
     maxAngularVelocity=3
     maxLinearAcceleration=300
     maxLinearVelocity=700
     velocityScaleZ=0
     externalLinearDragV=<0 0.5 0>
    }
   }
   skydive_GoingUp {simSpaceSettings {
     maxAngularAcceleration=1000
     maxAngularVelocity=3
     maxLinearAcceleration=300
     maxLinearVelocity=700
     velocityScaleZ=0
     externalLinearDragV=<0 0.5 0>
    }}
   skydive_Parachuting {
    componentLinearAccScale=<0 0.1 0.3>
    componentLinearVelScale=<0 0.1 0.3>
    simSpaceSettings {
     maxAngularAcceleration=1000
     maxAngularVelocity=3
     maxLinearAcceleration=300
     maxLinearVelocity=700
     velocityScaleZ=0
     externalLinearDragV=<0 0.5 0>
    }
   }
   skydive_Surfing {simSpaceSettings {
     maxAngularAcceleration=1000
     maxAngularVelocity=3
     maxLinearAcceleration=300
     maxLinearVelocity=700
     velocityScaleZ=0
     externalLinearDragV=<0 0.5 0>
    }}
   swimming {simSpaceSettings {
     maxAngularAcceleration=1000
     maxAngularVelocity=3
     maxLinearAcceleration=300
     maxLinearVelocity=700
     velocityScaleZ=0
     externalLinearDragV=<0 0.5 0>
    }}
   tactical_Sprint {simSpaceSettings {
     maxAngularAcceleration=1000
     maxAngularVelocity=3
     maxLinearAcceleration=300
     maxLinearVelocity=700
     velocityScaleZ=0
     externalLinearDragV=<0 0.5 0>
    }}
   ziplining {simSpaceSettings {
     maxAngularAcceleration=1000
     maxAngularVelocity=3
     maxLinearAcceleration=300
     maxLinearVelocity=700
     velocityScaleZ=0
     externalLinearDragV=<0 0.5 0>
    }}
  }
 }
]
