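# Physics asset for the ThornedEden female medium face accessory mesh, setting up secondary motion for two braids, earrings, a topknot and shoulder thorn straps.
# Capsules cover those dynamic bones plus head, neck, upper spine, clavicles and upper arms.
# Small base bodies join to the head or clavicles with free angular motion.
# From each base, tight swing and twist limits constrain the hanging pieces, looser swing for earrings, with braids as four-link chains.
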PhysicsAsset /BRCosmetics/Characters/Player/Female/Medium/Bodies/F_MED_ThornedEden/Meshes/Parts/F_MED_ThornedEden_FaceAcc_Physics
bodies {
 clavicle_l [{shapeType=Capsule, length=6, radius=6, shapeName=clavicle_l_capsule, center=<1.8251 0 0>, rotation=<rot 0 -90 90>}]
 clavicle_r [{shapeType=Capsule, length=5, radius=7.5, shapeName=clavicle_r_capsule, center=<-5 0 0>, rotation=<rot 0 90 90>}]
 dyn_braid_01_l [{shapeType=Capsule, length=1, radius=1, shapeName=dyn_braid_01_l_capsule, center=<2 0 0>, rotation=<rot 0 90 90>}]
 dyn_braid_01_l_base [{shapeType=Capsule, length=0.5, radius=0.5, shapeName=dyn_braid_01_l_base_capsule, rotation=<rot 0 -90 0>}]
 dyn_braid_01_r [{shapeType=Capsule, length=1, radius=1, shapeName=dyn_braid_01_r_capsule, center=<2 0 0>, rotation=<rot 0 90 90>}]
 dyn_braid_01_r_base [{shapeType=Capsule, length=0.5, radius=0.5, shapeName=dyn_braid_01_r_base_capsule, rotation=<rot 0 -90 0>}]
 dyn_braid_02_l [{shapeType=Capsule, length=1, radius=1, shapeName=dyn_braid_02_l_capsule, center=<2 0 0>, rotation=<rot 0 90 90>}]
 dyn_braid_02_r [{shapeType=Capsule, length=1, radius=1, shapeName=dyn_braid_02_r_capsule, center=<2 0 0>, rotation=<rot 0 90 90>}]
 dyn_braid_03_l [{shapeType=Capsule, length=1, radius=1, shapeName=dyn_braid_03_l_capsule, center=<2 0 0>, rotation=<rot 0 90 90>}]
 dyn_braid_03_r [{shapeType=Capsule, length=1, radius=1, shapeName=dyn_braid_03_r_capsule, center=<2 0 0>, rotation=<rot 0 90 90>}]
 dyn_braid_04_l [{shapeType=Capsule, length=3, radius=1, shapeName=dyn_braid_04_l_capsule, center=<3 0 0>, rotation=<rot 0 90 90>}]
 dyn_braid_04_r [{shapeType=Capsule, length=3, radius=1, shapeName=dyn_braid_04_r_capsule, center=<3 0 0>, rotation=<rot 0 90 90>}]
 dyn_earring_l [{shapeType=Capsule, length=6, radius=1, shapeName=dyn_earring_l_capsule, center=<3 0 0>, rotation=<rot 0 90 90>}]
 dyn_earring_l_base [{shapeType=Capsule, length=0.5, radius=0.5, shapeName=dyn_earring_l_base_capsule, rotation=<rot 0 -90 -90>}]
 dyn_earring_r [{shapeType=Capsule, length=6, radius=1, shapeName=dyn_earring_r_capsule, center=<3 0 0>, rotation=<rot 0 90 90>}]
 dyn_earring_r_base [{shapeType=Capsule, length=0.5, radius=0.5, shapeName=dyn_earring_r_base_capsule, rotation=<rot 0 -90 -90>}]
 dyn_thorn_strap_l [{shapeType=Capsule, length=5, radius=3, shapeName=dyn_thorn_strap_l_capsule, center=<8 0 0>}]
 dyn_thorn_strap_l_base [{shapeType=Capsule, length=0.505, radius=0.505, shapeName=dyn_thorn_strap_l_base_capsule, rotation=<rot 0 -90 0>}]
 dyn_thorn_strap_r [{shapeType=Capsule, length=5, radius=3, shapeName=dyn_thorn_strap_r_capsule, center=<8 0 0>}]
 dyn_thorn_strap_r_base [{shapeType=Capsule, length=0.505, radius=0.505, shapeName=dyn_thorn_strap_r_base_capsule, rotation=<rot 0 -90 0>}]
 dyn_topknot [{shapeType=Capsule, length=2, radius=4, shapeName=dyn_topknot_capsule, center=<6 0 0>, rotation=<rot 0 90 90>}]
 dyn_topknot_base [{shapeType=Capsule, length=0.505, radius=0.505, shapeName=dyn_topknot_base_capsule, rotation=<rot 0 -90 0>}]
 head [
  {shapeType=Capsule, length=10, radius=5, center=<2.2264 0 0>, rotation=<rot 0 -90 -90>}
  {shapeType=Capsule, length=5, radius=5.5, shapeName=head_capsule, center=<0 5 0>, rotation=<rot 0 90 90>}
 ]
 neck_01 [{shapeType=Capsule, length=5, radius=5, shapeName=neck_01_capsule, rotation=<rot 90 -90 90>}]
 spine_05 [{shapeType=Capsule, length=3.5, radius=9, shapeName=spine_05_capsule, center=<2.3281 2.9387 4.7141>, rotation=<rot 0 90 0>}]
 upperarm_l [{shapeType=Capsule, length=21.75, radius=7, shapeName=upperarm_l_capsule, center=<12.3273 0 0>, rotation=<rot 90 0 -0.0002>}]
 upperarm_r [
  {shapeType=Capsule, length=21.75, radius=3.5, shapeName=upperarm_r_capsule, center=<-12.3273 0 0>, rotation=<rot 90 180 179.9995>}
  {
   shapeType=Capsule
   length=5
   radius=4.5
   shapeName=clavicle_r_capsule
   center=<-4.5816 -1.1017 -1.1257>
   rotation=<rot -89.857 179.8207 -89.8209>
  }
 ]
}
constraints [
 {
  bone1Name=dyn_braid_01_l_base
  bone2Name=head
  swing1LimitDegrees=45
  swing1Motion=Free
  swing2LimitDegrees=45
  swing2Motion=Free
  twistLimitDegrees=45
  twistMotion=Free
 }
 {bone1Name=dyn_braid_01_l, bone2Name=dyn_braid_01_l_base, swing1LimitDegrees=15, swing2LimitDegrees=15, twistLimitDegrees=5}
 {bone1Name=dyn_braid_02_l, bone2Name=dyn_braid_01_l, swing1LimitDegrees=15, swing2LimitDegrees=15, twistLimitDegrees=5}
 {bone1Name=dyn_braid_03_l, bone2Name=dyn_braid_02_l, swing1LimitDegrees=15, swing2LimitDegrees=15, twistLimitDegrees=5}
 {bone1Name=dyn_braid_04_l, bone2Name=dyn_braid_03_l, swing1LimitDegrees=15, swing2LimitDegrees=15, twistLimitDegrees=5}
 {
  bone1Name=dyn_braid_01_r_base
  bone2Name=head
  swing1LimitDegrees=45
  swing1Motion=Free
  swing2LimitDegrees=45
  swing2Motion=Free
  twistLimitDegrees=45
  twistMotion=Free
 }
 {bone1Name=dyn_braid_01_r, bone2Name=dyn_braid_01_r_base, swing1LimitDegrees=15, swing2LimitDegrees=15, twistLimitDegrees=5}
 {bone1Name=dyn_braid_02_r, bone2Name=dyn_braid_01_r, swing1LimitDegrees=15, swing2LimitDegrees=15, twistLimitDegrees=5}
 {bone1Name=dyn_braid_03_r, bone2Name=dyn_braid_02_r, swing1LimitDegrees=15, swing2LimitDegrees=15, twistLimitDegrees=5}
 {bone1Name=dyn_braid_04_r, bone2Name=dyn_braid_03_r, swing1LimitDegrees=15, swing2LimitDegrees=15, twistLimitDegrees=5}
 {bone1Name=dyn_earring_r, bone2Name=dyn_earring_r_base, swing1LimitDegrees=45, swing2LimitDegrees=45, twistLimitDegrees=5}
 {
  bone1Name=dyn_earring_l_base
  bone2Name=head
  swing1LimitDegrees=45
  swing1Motion=Free
  swing2LimitDegrees=45
  swing2Motion=Free
  twistLimitDegrees=45
  twistMotion=Free
 }
 {bone1Name=dyn_earring_l, bone2Name=dyn_earring_l_base, swing1LimitDegrees=45, swing2LimitDegrees=45, twistLimitDegrees=5}
 {
  bone1Name=dyn_earring_r_base
  bone2Name=head
  swing1LimitDegrees=45
  swing1Motion=Free
  swing2LimitDegrees=45
  swing2Motion=Free
  twistLimitDegrees=45
  twistMotion=Free
 }
 {bone1Name=dyn_topknot, bone2Name=dyn_topknot_base, swing1LimitDegrees=10, swing2LimitDegrees=10, twistLimitDegrees=5}
 {
  bone1Name=dyn_topknot_base
  bone2Name=head
  swing1LimitDegrees=45
  swing1Motion=Free
  swing2LimitDegrees=45
  swing2Motion=Free
  twistLimitDegrees=45
  twistMotion=Free
 }
 {
  bone1Name=upperarm_r
  bone2Name=clavicle_r
  swing1LimitDegrees=45
  swing1Motion=Free
  swing2LimitDegrees=45
  swing2Motion=Free
  twistLimitDegrees=45
  twistMotion=Free
 }
 {
  bone1Name=upperarm_l
  bone2Name=clavicle_l
  swing1LimitDegrees=45
  swing1Motion=Free
  swing2LimitDegrees=45
  swing2Motion=Free
  twistLimitDegrees=45
  twistMotion=Free
 }
 {
  bone1Name=head
  bone2Name=spine_05
  swing1LimitDegrees=45
  swing1Motion=Free
  swing2LimitDegrees=45
  swing2Motion=Free
  twistLimitDegrees=45
  twistMotion=Free
 }
 {
  bone1Name=clavicle_r
  bone2Name=spine_05
  swing1LimitDegrees=45
  swing1Motion=Free
  swing2LimitDegrees=45
  swing2Motion=Free
  twistLimitDegrees=45
  twistMotion=Free
 }
 {
  bone1Name=clavicle_l
  bone2Name=spine_05
  swing1LimitDegrees=45
  swing1Motion=Free
  swing2LimitDegrees=45
  swing2Motion=Free
  twistLimitDegrees=45
  twistMotion=Free
 }
 {
  bone1Name=dyn_thorn_strap_l_base
  bone2Name=clavicle_l
  swing1LimitDegrees=45
  swing1Motion=Free
  swing2LimitDegrees=45
  swing2Motion=Free
  twistLimitDegrees=45
  twistMotion=Free
 }
 {bone1Name=dyn_thorn_strap_l, bone2Name=dyn_thorn_strap_l_base, swing1LimitDegrees=15, swing2LimitDegrees=5, twistLimitDegrees=5}
 {bone1Name=dyn_thorn_strap_r, bone2Name=dyn_thorn_strap_r_base, swing1LimitDegrees=15, swing2LimitDegrees=5, twistLimitDegrees=5}
 {
  bone1Name=dyn_thorn_strap_r_base
  bone2Name=clavicle_r
  swing1LimitDegrees=45
  swing1Motion=Free
  swing2LimitDegrees=45
  swing2Motion=Free
  twistLimitDegrees=45
  twistMotion=Free
 }
 {
  bone1Name=neck_01
  bone2Name=spine_05
  swing1LimitDegrees=45
  swing1Motion=Free
  swing2LimitDegrees=45
  swing2Motion=Free
  twistLimitDegrees=45
  twistMotion=Free
 }
]
