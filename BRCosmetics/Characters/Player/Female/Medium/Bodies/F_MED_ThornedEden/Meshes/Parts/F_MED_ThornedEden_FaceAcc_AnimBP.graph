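# Animation Blueprint for the ThornedEden female character's face accessory, adding head-relative physics and piercing motion.
# It copies the attached parent's pose, layering upper-arm branches from a body mesh cached when character customization loads.
# Two RigidBody simulations, using front-end and standard physics assets, feed a bool blend.
# A weighted constraint then makes extra_piercing follow nose, cheek and head bones before a Control Rig finishes the pose.
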AnimBlueprint /BRCosmetics/Characters/Player/Female/Medium/Bodies/F_MED_ThornedEden/Meshes/Parts/F_MED_ThornedEden_FaceAcc_AnimBP
target_skeleton=&/F_MED_ThornedEden_FaceAcc_Skeleton
graphs [
 AnimGraph {nodes [
   n1 = RigidBody(ComponentPose=Local To Component(LocalPose=Use cached pose 'pre_dyn'())) {
    bOverrideWorldGravity=true
    overridePhysicsAsset=&/F_MED_ThornedEden_FaceAcc_Physics_FrontEnd
    simulationSpace=BaseBoneSpace
    alphaBoolBlend {blendOption=Linear}
    baseBoneRef {boneName=head}
    componentAppliedLinearAccClamp=<3000 3000 3000>
    componentLinearAccScale=<0.4 0.4 0.4>
    componentLinearVelScale=<0.4 0.4 0.4>
    overrideWorldGravity {z=-980}
    simSpaceSettings {maxAngularAcceleration=100, maxAngularVelocity=50, maxLinearAcceleration=1, maxLinearVelocity=10, velocityScaleZ=0, worldAlpha=1}
   }
   n2 = Component To Local(ComponentPose=@n1)
   n3 = RigidBody(ComponentPose=Local To Component(LocalPose=Use cached pose 'pre_dyn'())) {
    bOverrideWorldGravity=true
    overridePhysicsAsset=&/F_MED_ThornedEden_FaceAcc_Physics
    simulationSpace=BaseBoneSpace
    alphaBoolBlend {blendOption=Linear}
    baseBoneRef {boneName=head}
    componentAppliedLinearAccClamp=<3000 3000 3000>
    componentLinearAccScale=<0.4 0.4 0.4>
    componentLinearVelScale=<0.4 0.4 0.4>
    overrideWorldGravity {z=-980}
    simSpaceSettings {maxAngularAcceleration=100, maxAngularVelocity=50, maxLinearAcceleration=1, maxLinearVelocity=10, velocityScaleZ=0, worldAlpha=1}
   }
   n4 = Component To Local(ComponentPose=@n3)
   n5 = "Constraint - extra_piercing (PPPPP)"!(
    ComponentPose=Local To Component(LocalPose=Blend Poses by bool(BlendPose_0=@n2, BlendPose_1=@n4) {blendType=Linear})
    Alpha=1
    ConstraintWeights_0=0.6157
    ConstraintWeights_1:float=0.1563
    ConstraintWeights_2:float=0.1533
    ConstraintWeights_3:float=0.0493
    ConstraintWeights_4:float=0.0254
   ) -> (Pose) {
    alphaBoolBlend {blendOption=Linear}
    boneToModify {boneName=extra_piercing}
    constraintSetup [
     {targetBone {boneName=FACIAL_R_Nostril}}
     {targetBone {boneName=FACIAL_R_CheekInner}}
     {targetBone {boneName=FACIAL_C_Nose}}
     {targetBone {boneName=head}}
     {targetBone {boneName=FACIAL_C_NoseTip}}
    ]
    constraintWeights [1,1,1,1,1]
   }
   n6 = Control Rig(Source=Component To Local(ComponentPose=@n5)) {
    alphaBoolBlend {blendOption=Linear}
    controlRigAssetReference {blueprintRigClass=&/F_MED_ThornedEden_FaceAcc_CtrlRig.F_MED_ThornedEden_FaceAcc_CtrlRig_C}
   }
   Output Pose(Result=@n6) {layerGroup=None}
   n7 = Get BodyMesh() -> (BodyMesh:Skeletal Mesh Component Object Reference)
   Save cached pose 'pre_dyn'(Pose=Layered blend per bone(
     BasePose=Copy Pose From Mesh() {bUseAttachedParent=true}
     BlendPoses_0=Copy Pose From Mesh(SourceMeshComponent=@n7) {bUseAttachedParent=true}
    ) {layerSetup [{branchFilters [{boneName=upperarm_l}, {boneName=upperarm_r}]}]}) {cacheName=pre_dyn}
  ]}
 EventGraph {nodes [
   n1 = Event OnCharacterCustomizationLoaded()
   > Set BodyMesh(BodyMesh="Get (a copy)"(Array:Array of Skeletal Mesh Component Budgeted Object References=@n1.CustomMeshParts, Dimension 1=1) -> (Output:Skeletal Mesh Component Budgeted Object Reference))
   Event BlueprintUpdateAnimation()
  ]}
]
variables [BodyMesh {type=SkeletalMeshComponent}]
